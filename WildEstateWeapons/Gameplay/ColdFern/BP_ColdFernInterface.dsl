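; BP_ColdFernInterface is the Blueprint interface for the ColdFern weapon's special shot state.
; GetSpecialShotData returns a boolean, false by default, and SetSpecialShotData takes an IsDoingSpecialShot boolean.
; Neither function carries logic here, so storing and reacting to that state is left to implementers.
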(fn GetSpecialShotData
  (return false))

(fn SetSpecialShotData
  (params (IsDoingSpecialShot :type Boolean)))
