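# Animation Blueprint for the ColdFern shotgun weapon mesh, targeting the ColdFern shotgun skeleton.
# A non-looping sequence player supplies the weapon's pose animation as the base pose.
# That pose feeds the DefaultSlot node, so montages can play over it before reaching the output.
# The event graph is empty and the Blueprint declares no variables.
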AnimBlueprint /WildEstateWeapons/Gameplay/ColdFern/ColdFern_Wep_AnimBP
target_skeleton=/WildEstateWeapons/Assets/ColdFern_Shotgun/Meshes/ColdFern_Shotgun_Skeleton
graphs [
 AnimGraph {nodes [Output Pose(Result=Slot 'DefaultSlot'(Source=Sequence Player 'WildEstate_ColdFern_Shotgun_Pose_WEP'() {bLoopAnimation=false, sequence=/WildEstateWeapons/Assets/ColdFern_Shotgun/Animation/WildEstate_ColdFern_Shotgun_Pose_WEP}))]}
 EventGraph {links [], nodes []}
]
variables []
