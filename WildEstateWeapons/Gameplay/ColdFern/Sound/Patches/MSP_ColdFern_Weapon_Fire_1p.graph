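# MetaSound patch that plays ColdFern's first-person gunshot Sound in stereo with random volume and an input-set pitch.
# OnGunshot either retriggers one voice or rotates through three via Trigger Sequence, depending on bUseSingleVoice; a fourth player is mixed in but never triggered.
# Trigger Control gates let only the most recently started voice fire OnFinished.
# The Mobile page uses a single wave player.
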MetaSoundPatch /WildEstateWeapons/Gameplay/ColdFern/Sound/Patches/MSP_ColdFern_Weapon_Fire_1p
inputs [
 bUseSingleVoice {type=Bool}
 OnGunshot {type=Trigger}
 Random.Pitch {type=Float}
 Random.Volume.Max {type=Float}
 Random.Volume.Min {type=Float}
 Sound {type=WaveAsset}
]
outputs [
 OnFinished {type=Trigger}
 Out.L {type=Audio}
 Out.R {type=Audio}
]
pages [
 Default {
  is_build_graph=true
  input_defaults {
   OnGunshot=null
   Random.Pitch=0
   Random.Volume.Max=0
   Random.Volume.Min=0
   Sound=null
   bUseSingleVoice="<MetaSoundBuilderResult.SUCCEEDED: 0>"
  }
  nodes [
   n1 = TriggerCompare.Clamp:Bool(Compare=^OnGunshot, A=^bUseSingleVoice) {also {defaults {B="<MetaSoundBuilderResult.SUCCEEDED: 0>"}}}
   TriggerSequence3_0 = "UE.Trigger Sequence (3)"(In=@n1.False)
   TriggerAnyTriggerAny2_0 = "TriggerAny.Trigger Any (2)"(In 0=@n1.True, In 1=@TriggerSequence3_0.Out 0)
   n2 = VariableAccessor.Trigger(Variable=VariableMutator.Trigger(Variable=InitVariable.Trigger().Variable, Value=@TriggerAnyTriggerAny2_0.Out).Variable)
   n3 = UE.RandomFloat(Max=^Random.Volume.Max, Min=^Random.Volume.Min, Next=@n2.Value)
   FrequencyMultipliertoSemitone_0 = UE.Frequency Multiplier to Semitone:Float(Frequency Multiplier=^Random.Pitch)
   WavePlayer_0 = UE.Wave Player:Stereo(Wave Asset=^Sound, Pitch Shift=@FrequencyMultipliertoSemitone_0.Semitones, Play=VariableAccessor.Trigger(Variable=@n2.Variable).Value) {also {version [1,1]}}
   n4 = UE.Multiply:Audio by Float(AdditionalOperands=@n3.Value, PrimaryOperand=@WavePlayer_0.Out Left)
   n5 = VariableAccessor.Trigger(Variable=VariableMutator.Trigger(Variable=InitVariable.Trigger().Variable, Value=@TriggerSequence3_0.Out 1).Variable)
   n6 = UE.RandomFloat(Max=^Random.Volume.Max, Min=^Random.Volume.Min, Next=@n5.Value)
   WavePlayer_3 = UE.Wave Player:Stereo(Wave Asset=^Sound, Pitch Shift=@FrequencyMultipliertoSemitone_0.Semitones, Play=VariableAccessor.Trigger(Variable=@n5.Variable).Value) {also {version [1,1]}}
   n7 = UE.Multiply:Audio by Float(AdditionalOperands=@n6.Value, PrimaryOperand=@WavePlayer_3.Out Left)
   n8 = VariableAccessor.Trigger(Variable=VariableMutator.Trigger(Variable=InitVariable.Trigger().Variable, Value=@TriggerSequence3_0.Out 2).Variable)
   n9 = UE.RandomFloat(Max=^Random.Volume.Max, Min=^Random.Volume.Min, Next=VariableAccessor.Trigger(Variable=@n8.Variable).Value)
   WavePlayer_1 = UE.Wave Player:Stereo(Wave Asset=^Sound, Pitch Shift=@FrequencyMultipliertoSemitone_0.Semitones, Play=@n8.Value) {also {version [1,1]}}
   Multiply_1 = UE.Multiply:Audio by Float(AdditionalOperands=@n9.Value, PrimaryOperand=@WavePlayer_1.Out Left)
   n10 = VariableAccessor.Trigger(Variable=InitVariable.Trigger().Variable)
   n11 = UE.RandomFloat(Max=^Random.Volume.Max, Min=^Random.Volume.Min, Next=VariableAccessor.Trigger(Variable=@n10.Variable).Value)
   WavePlayer_2 = UE.Wave Player:Stereo(Wave Asset=^Sound, Pitch Shift=@FrequencyMultipliertoSemitone_0.Semitones, Play=@n10.Value) {also {version [1,1]}}
   Add_0 = UE.Add:Audio(
    PrimaryOperand=UE.Add:Audio(PrimaryOperand=UE.Add:Audio(PrimaryOperand=@n4.Out, AdditionalOperands=@n7.Out).Out, AdditionalOperands=@Multiply_1.Out).Out
    AdditionalOperands=UE.Multiply:Audio by Float(AdditionalOperands=@n11.Value, PrimaryOperand=@WavePlayer_2.Out Left).Out
   )
   n12 = UE.Multiply:Audio by Float(AdditionalOperands=@n6.Value, PrimaryOperand=@WavePlayer_3.Out Right)
   n13 = UE.Multiply:Audio by Float(AdditionalOperands=@n3.Value, PrimaryOperand=@WavePlayer_0.Out Right)
   Multiply_0 = UE.Multiply:Audio by Float(AdditionalOperands=@n9.Value, PrimaryOperand=@WavePlayer_1.Out Right)
   Add_4 = UE.Add:Audio(
    PrimaryOperand=UE.Add:Audio(PrimaryOperand=UE.Add:Audio(AdditionalOperands=@n12.Out, PrimaryOperand=@n13.Out).Out, AdditionalOperands=@Multiply_0.Out).Out
    AdditionalOperands=UE.Multiply:Audio by Float(AdditionalOperands=@n11.Value, PrimaryOperand=@WavePlayer_2.Out Right).Out
   )
   InitVariable.Float()
   InitVariable.Float()
   InitVariable.Float()
   InitVariable.Float()
   n14 = VariableAccessor.Trigger(Variable=VariableMutator.Trigger(Variable=InitVariable.Trigger().Variable, Value=@WavePlayer_2.On Play).Variable)
   n15 = VariableAccessor.Trigger(Variable=@n14.Variable)
   n16 = VariableAccessor.Trigger(Variable=VariableMutator.Trigger(Variable=InitVariable.Trigger().Variable, Value=@WavePlayer_0.On Play).Variable)
   n17 = VariableAccessor.Trigger(Variable=@n16.Variable)
   n18 = VariableAccessor.Trigger(Variable=VariableMutator.Trigger(Variable=InitVariable.Trigger().Variable, Value=@WavePlayer_1.On Play).Variable)
   n19 = VariableAccessor.Trigger(Variable=@n18.Variable)
   n20 = VariableAccessor.Trigger(Variable=@n19.Variable)
   TriggerAnyTriggerAny3_0 = "TriggerAny.Trigger Any (3)"(In 2=@n15.Value, In 0=@n17.Value, In 1=@n20.Value)
   n21 = VariableAccessor.Trigger(Variable=VariableMutator.Trigger(Variable=InitVariable.Trigger().Variable, Value=@WavePlayer_3.On Play).Variable)
   n22 = InitVariable.Trigger()
   TriggerControl_0 = UE.Trigger Control(
    Close=@TriggerAnyTriggerAny3_0.Out
    Open=@n21.Value
    Trigger In=VariableAccessor.Trigger(Variable=VariableMutator.Trigger(Variable=@n22.Variable, Value=@WavePlayer_3.On Finished).Variable).Value
   )
   n23 = VariableAccessor.Trigger(Variable=@n17.Variable)
   n24 = VariableAccessor.Trigger(Variable=@n21.Variable)
   n25 = VariableAccessor.Trigger(Variable=@n24.Variable)
   TriggerAnyTriggerAny3_1 = "TriggerAny.Trigger Any (3)"(
    In 2=VariableAccessor.Trigger(Variable=@n20.Variable).Value
    In 0=VariableAccessor.Trigger(Variable=@n23.Variable).Value
    In 1=VariableAccessor.Trigger(Variable=@n25.Variable).Value
   )
   n26 = InitVariable.Trigger()
   TriggerControl_1 = UE.Trigger Control(
    Close=@TriggerAnyTriggerAny3_1.Out
    Open=@n14.Value
    Trigger In=VariableAccessor.Trigger(Variable=VariableMutator.Trigger(Variable=@n26.Variable, Value=@WavePlayer_2.On Finished).Variable).Value
   )
   n27 = VariableAccessor.Trigger(Variable=@n15.Variable)
   TriggerAnyTriggerAny3_2 = "TriggerAny.Trigger Any (3)"(In 0=@n24.Value, In 2=VariableAccessor.Trigger(Variable=@n27.Variable).Value, In 1=@n19.Value)
   n28 = InitVariable.Trigger()
   TriggerControl_2 = UE.Trigger Control(
    Close=@TriggerAnyTriggerAny3_2.Out
    Open=@n16.Value
    Trigger In=VariableAccessor.Trigger(Variable=VariableMutator.Trigger(Variable=@n28.Variable, Value=@WavePlayer_0.On Finished).Variable).Value
   )
   TriggerAnyTriggerAny3_3 = "TriggerAny.Trigger Any (3)"(In 2=@n27.Value, In 0=@n23.Value, In 1=@n25.Value)
   n29 = InitVariable.Trigger()
   TriggerControl_3 = UE.Trigger Control(
    Close=@TriggerAnyTriggerAny3_3.Out
    Open=@n18.Value
    Trigger In=VariableAccessor.Trigger(Variable=VariableMutator.Trigger(Variable=@n29.Variable, Value=@WavePlayer_1.On Finished).Variable).Value
   )
   TriggerAnyTriggerAny4_0 = "TriggerAny.Trigger Any (4)"(
    In 1=@TriggerControl_0.Trigger Out
    In 3=@TriggerControl_1.Trigger Out
    In 0=@TriggerControl_2.Trigger Out
    In 2=@TriggerControl_3.Trigger Out
   )
  ]
  links [{dst="@Out.L", src=Add_0.Out}, {dst="@Out.R", src=Add_4.Out}, {dst="@OnFinished", src=TriggerAnyTriggerAny4_0.Out}]
 }
 Mobile {
  nodes [
   InitVariable.Trigger()
   InitVariable.Trigger()
   n1 = UE.RandomFloat(Next=^OnGunshot, Max=^Random.Volume.Max, Min=^Random.Volume.Min)
   n2 = UE.RandomFloat(Next=^OnGunshot, Max=^Random.Pitch, Min=^Random.Pitch)
   FrequencyMultipliertoSemitone_0 = UE.Frequency Multiplier to Semitone:Float(Frequency Multiplier=@n2.Value)
   WavePlayer_0 = UE.Wave Player:Stereo(Wave Asset=^Sound, Pitch Shift=@FrequencyMultipliertoSemitone_0.Semitones, Play=@n2.On Next) {also {version [1,1]}}
   Multiply_0 = UE.Multiply:Audio by Float(AdditionalOperands=@n1.Value, PrimaryOperand=@WavePlayer_0.Out Left)
   Multiply_1 = UE.Multiply:Audio by Float(AdditionalOperands=@n1.Value, PrimaryOperand=@WavePlayer_0.Out Right)
  ]
  links [{dst="@Out.L", src=Multiply_0.Out}, {dst="@Out.R", src=Multiply_1.Out}, {dst="@OnFinished", src=WavePlayer_0.On Finished}]
 }
]
