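# MetaSound patch that plays ColdFern third-person stereo gunshots from an indexed wave array, with random volume and input-driven pitch.
# Shots cycle through three wave-player voices, or use one when bUseSingleVoice is set; a fourth voice's play trigger is unfed.
# Decay envelopes duck each voice to a floor unless a probability roll lets it ring out, and only the latest voice fires OnFinished.
# A Mobile page uses one player without envelopes or a connected wave.
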MetaSoundPatch /WildEstateWeapons/Gameplay/ColdFern/Sound/Patches/MSP_ColdFern_Weapon_Fire_3p_Stereo
inputs [
 bUseSingleVoice {type=Bool}
 Envelope.A.SustainTime {type=Time}
 Envelope.A.SustainVolume {type=Float}
 Envelope.B.SustainTime {type=Time}
 Envelope.B.SustainVolume {type=Float}
 Envelope.C.SustainTime {type=Time}
 Envelope.C.SustainVolume {type=Float}
 Envelope.D.SustainTime {type=Time}
 Envelope.D.SustainVolume {type=Float}
 OnGunshot {type=Trigger}
 Random.Pitch {type=Float}
 Random.Volume.Max {type=Float}
 Random.Volume.Min {type=Float}
 RingOutProbability {type=Float}
 Sound {type=WaveAsset:Array}
 SoundIndex {type=Int32}
]
outputs [
 OnFinished {type=Trigger}
 Out.L {type=Audio}
 Out.R {type=Audio}
]
pages [
 Default {
  is_build_graph=true
  input_defaults {
   Envelope.A.SustainTime=0.8
   Envelope.A.SustainVolume=0
   Envelope.B.SustainTime=0.8
   Envelope.B.SustainVolume=0
   Envelope.C.SustainTime=0.8
   Envelope.C.SustainVolume=0
   Envelope.D.SustainTime=0.8
   Envelope.D.SustainVolume=0
   OnGunshot=null
   Random.Pitch=0
   Random.Volume.Max=0
   Random.Volume.Min=0
   RingOutProbability=0.05
   SoundIndex=0
   bUseSingleVoice="<MetaSoundBuilderResult.SUCCEEDED: 0>"
   Sound []
  }
  nodes [
   n1 = TriggerCompare.Clamp:Bool(Compare=^OnGunshot, A=^bUseSingleVoice) {also {defaults {B="<MetaSoundBuilderResult.SUCCEEDED: 0>"}}}
   TriggerSequence3_0 = "UE.Trigger Sequence (3)"(In=@n1.False)
   TriggerAnyTriggerAny2_0 = "TriggerAny.Trigger Any (2)"(In 0=@n1.True, In 1=@TriggerSequence3_0.Out 0)
   n2 = VariableAccessor.Trigger(Variable=VariableMutator.Trigger(Variable=InitVariable.Trigger().Variable, Value=@TriggerAnyTriggerAny2_0.Out).Variable)
   n3 = InitVariable.Float()
   n4 = VariableAccessor.Trigger(Variable=@n2.Variable)
   ADEnvelopeADEnvelope_1 = AD Envelope.AD Envelope:Float(Decay Time=^Envelope.A.SustainTime, Trigger=VariableAccessor.Trigger(Variable=@n4.Variable).Value) {also {defaults {Attack Time=0}}}
   TriggerFilter_0 = UE.Trigger Filter(Trigger=^OnGunshot, Probability=^RingOutProbability)
   TriggerRouteTriggerRouteFloat2_2 = "TriggerRoute.Trigger Route (Float, 2):Float"(
    Value 1=Max.Max:Float(B=^Envelope.A.SustainVolume, A=@ADEnvelopeADEnvelope_1.Out Envelope).Value
    Set 0=@TriggerFilter_0.Heads
    Set 1=@TriggerFilter_0.Tails
   ) {also {defaults {Value 0=1}}}
   n5 = UE.Multiply:Float(
    PrimaryOperand=UE.RandomFloat(Max=^Random.Volume.Max, Min=^Random.Volume.Min, Next=@n2.Value).Value
    AdditionalOperands=VariableAccessor.Float(Variable=VariableMutator.Float(Variable=@n3.Variable, Value=@TriggerRouteTriggerRouteFloat2_2.Value).Variable).Value
   )
   n6 = Array.Get:WaveAsset:Array(Trigger=^OnGunshot, Array=^Sound, Index=^SoundIndex)
   FrequencyMultipliertoSemitone_0 = UE.Frequency Multiplier to Semitone:Float(Frequency Multiplier=^Random.Pitch)
   n7 = VariableAccessor.Float(Variable=VariableMutator.Float(Value=@FrequencyMultipliertoSemitone_0.Semitones, Variable=InitVariable.Float().Variable).Variable)
   n8 = VariableAccessor.Float(Variable=@n7.Variable)
   WavePlayer_0 = UE.Wave Player:Stereo(Wave Asset=@n6.Element, Pitch Shift=@n8.Value, Play=@n4.Value) {also {version [1,1]}}
   n9 = UE.Multiply:Audio by Float(AdditionalOperands=@n5.Out, PrimaryOperand=@WavePlayer_0.Out Left)
   n10 = VariableAccessor.Trigger(Variable=VariableMutator.Trigger(Variable=InitVariable.Trigger().Variable, Value=@TriggerSequence3_0.Out 1).Variable)
   n11 = InitVariable.Float()
   n12 = VariableAccessor.Trigger(Variable=@n10.Variable)
   ADEnvelopeADEnvelope_2 = AD Envelope.AD Envelope:Float(Decay Time=^Envelope.B.SustainTime, Trigger=VariableAccessor.Trigger(Variable=@n12.Variable).Value) {also {defaults {Attack Time=0}}}
   TriggerFilter_1 = UE.Trigger Filter(Trigger=^OnGunshot, Probability=^RingOutProbability)
   TriggerRouteTriggerRouteFloat2_1 = "TriggerRoute.Trigger Route (Float, 2):Float"(
    Value 1=Max.Max:Float(B=^Envelope.B.SustainVolume, A=@ADEnvelopeADEnvelope_2.Out Envelope).Value
    Set 0=@TriggerFilter_1.Heads
    Set 1=@TriggerFilter_1.Tails
   ) {also {defaults {Value 0=1}}}
   n13 = UE.Multiply:Float(
    PrimaryOperand=UE.RandomFloat(Max=^Random.Volume.Max, Min=^Random.Volume.Min, Next=@n10.Value).Value
    AdditionalOperands=VariableAccessor.Float(Variable=VariableMutator.Float(Variable=@n11.Variable, Value=@TriggerRouteTriggerRouteFloat2_1.Value).Variable).Value
   )
   WavePlayer_3 = UE.Wave Player:Stereo(Wave Asset=UE.Reroute(Value=@n6.Element).Value, Pitch Shift=@n7.Value, Play=@n12.Value) {also {version [1,1]}}
   n14 = UE.Multiply:Audio by Float(AdditionalOperands=@n13.Out, PrimaryOperand=@WavePlayer_3.Out Left)
   n15 = VariableAccessor.Trigger(Variable=VariableMutator.Trigger(Variable=InitVariable.Trigger().Variable, Value=@TriggerSequence3_0.Out 2).Variable)
   n16 = VariableAccessor.Trigger(Variable=@n15.Variable)
   n17 = InitVariable.Float()
   ADEnvelopeADEnvelope_3 = AD Envelope.AD Envelope:Float(Trigger=VariableAccessor.Trigger(Variable=@n16.Variable).Value) {also {defaults {Attack Time=0, Decay Time=0.3}}}
   TriggerFilter_3 = UE.Trigger Filter(Trigger=^OnGunshot, Probability=^RingOutProbability)
   TriggerRouteTriggerRouteFloat2_0 = "TriggerRoute.Trigger Route (Float, 2):Float"(
    Value 1=Max.Max:Float(B=^Envelope.C.SustainVolume, A=@ADEnvelopeADEnvelope_3.Out Envelope).Value
    Set 0=@TriggerFilter_3.Heads
    Set 1=@TriggerFilter_3.Tails
   ) {also {defaults {Value 0=1}}}
   n18 = UE.Multiply:Float(
    PrimaryOperand=UE.RandomFloat(Max=^Random.Volume.Max, Min=^Random.Volume.Min, Next=@n16.Value).Value
    AdditionalOperands=VariableAccessor.Float(Variable=VariableMutator.Float(Variable=@n17.Variable, Value=@TriggerRouteTriggerRouteFloat2_0.Value).Variable).Value
   )
   n19 = VariableAccessor.Float(Variable=@n8.Variable)
   WavePlayer_1 = UE.Wave Player:Stereo(Wave Asset=@n6.Element, Pitch Shift=@n19.Value, Play=@n15.Value) {also {version [1,1]}}
   Multiply_1 = UE.Multiply:Audio by Float(AdditionalOperands=@n18.Out, PrimaryOperand=@WavePlayer_1.Out Left)
   n20 = VariableAccessor.Trigger(Variable=InitVariable.Trigger().Variable)
   n21 = VariableAccessor.Trigger(Variable=@n20.Variable)
   n22 = InitVariable.Float()
   ADEnvelopeADEnvelope_0 = AD Envelope.AD Envelope:Float(Trigger=VariableAccessor.Trigger(Variable=@n21.Variable).Value) {also {defaults {Attack Time=0, Decay Time=0.3}}}
   TriggerFilter_2 = UE.Trigger Filter(Trigger=^OnGunshot, Probability=^RingOutProbability)
   TriggerRouteTriggerRouteFloat2_3 = "TriggerRoute.Trigger Route (Float, 2):Float"(
    Value 1=Max.Max:Float(B=^Envelope.D.SustainVolume, A=@ADEnvelopeADEnvelope_0.Out Envelope).Value
    Set 0=@TriggerFilter_2.Heads
    Set 1=@TriggerFilter_2.Tails
   ) {also {defaults {Value 0=1}}}
   n23 = UE.Multiply:Float(
    PrimaryOperand=UE.RandomFloat(Max=^Random.Volume.Max, Min=^Random.Volume.Min, Next=@n21.Value).Value
    AdditionalOperands=VariableAccessor.Float(Variable=VariableMutator.Float(Variable=@n22.Variable, Value=@TriggerRouteTriggerRouteFloat2_3.Value).Variable).Value
   )
   WavePlayer_2 = UE.Wave Player:Stereo(Wave Asset=@n6.Element, Pitch Shift=VariableAccessor.Float(Variable=@n19.Variable).Value, Play=@n20.Value) {also {version [1,1]}}
   Add_0 = UE.Add:Audio(
    PrimaryOperand=UE.Add:Audio(PrimaryOperand=UE.Add:Audio(PrimaryOperand=@n9.Out, AdditionalOperands=@n14.Out).Out, AdditionalOperands=@Multiply_1.Out).Out
    AdditionalOperands=UE.Multiply:Audio by Float(AdditionalOperands=@n23.Out, PrimaryOperand=@WavePlayer_2.Out Left).Out
   )
   n24 = UE.Multiply:Audio by Float(AdditionalOperands=@n13.Out, PrimaryOperand=@WavePlayer_3.Out Right)
   n25 = UE.Multiply:Audio by Float(AdditionalOperands=@n5.Out, PrimaryOperand=@WavePlayer_0.Out Right)
   Multiply_0 = UE.Multiply:Audio by Float(AdditionalOperands=@n18.Out, PrimaryOperand=@WavePlayer_1.Out Right)
   Add_4 = UE.Add:Audio(
    PrimaryOperand=UE.Add:Audio(PrimaryOperand=UE.Add:Audio(AdditionalOperands=@n24.Out, PrimaryOperand=@n25.Out).Out, AdditionalOperands=@Multiply_0.Out).Out
    AdditionalOperands=UE.Multiply:Audio by Float(AdditionalOperands=@n23.Out, PrimaryOperand=@WavePlayer_2.Out Right).Out
   )
   n26 = VariableAccessor.Trigger(Variable=VariableMutator.Trigger(Variable=InitVariable.Trigger().Variable, Value=@WavePlayer_2.On Play).Variable)
   n27 = VariableAccessor.Trigger(Variable=@n26.Variable)
   n28 = VariableAccessor.Trigger(Variable=VariableMutator.Trigger(Variable=InitVariable.Trigger().Variable, Value=@WavePlayer_0.On Play).Variable)
   n29 = VariableAccessor.Trigger(Variable=@n28.Variable)
   n30 = VariableAccessor.Trigger(Variable=VariableMutator.Trigger(Variable=InitVariable.Trigger().Variable, Value=@WavePlayer_1.On Play).Variable)
   n31 = VariableAccessor.Trigger(Variable=@n30.Variable)
   n32 = VariableAccessor.Trigger(Variable=@n31.Variable)
   TriggerAnyTriggerAny3_0 = "TriggerAny.Trigger Any (3)"(In 2=@n27.Value, In 0=@n29.Value, In 1=@n32.Value)
   n33 = VariableAccessor.Trigger(Variable=VariableMutator.Trigger(Variable=InitVariable.Trigger().Variable, Value=@WavePlayer_3.On Play).Variable)
   n34 = InitVariable.Trigger()
   TriggerControl_0 = UE.Trigger Control(
    Close=@TriggerAnyTriggerAny3_0.Out
    Open=@n33.Value
    Trigger In=VariableAccessor.Trigger(Variable=VariableMutator.Trigger(Variable=@n34.Variable, Value=@WavePlayer_3.On Finished).Variable).Value
   )
   n35 = VariableAccessor.Trigger(Variable=@n29.Variable)
   n36 = VariableAccessor.Trigger(Variable=@n33.Variable)
   n37 = VariableAccessor.Trigger(Variable=@n36.Variable)
   TriggerAnyTriggerAny3_1 = "TriggerAny.Trigger Any (3)"(
    In 2=VariableAccessor.Trigger(Variable=@n32.Variable).Value
    In 0=VariableAccessor.Trigger(Variable=@n35.Variable).Value
    In 1=VariableAccessor.Trigger(Variable=@n37.Variable).Value
   )
   n38 = InitVariable.Trigger()
   TriggerControl_1 = UE.Trigger Control(
    Close=@TriggerAnyTriggerAny3_1.Out
    Open=@n26.Value
    Trigger In=VariableAccessor.Trigger(Variable=VariableMutator.Trigger(Variable=@n38.Variable, Value=@WavePlayer_2.On Finished).Variable).Value
   )
   n39 = VariableAccessor.Trigger(Variable=@n27.Variable)
   TriggerAnyTriggerAny3_2 = "TriggerAny.Trigger Any (3)"(In 0=@n36.Value, In 2=VariableAccessor.Trigger(Variable=@n39.Variable).Value, In 1=@n31.Value)
   n40 = InitVariable.Trigger()
   TriggerControl_2 = UE.Trigger Control(
    Close=@TriggerAnyTriggerAny3_2.Out
    Open=@n28.Value
    Trigger In=VariableAccessor.Trigger(Variable=VariableMutator.Trigger(Variable=@n40.Variable, Value=@WavePlayer_0.On Finished).Variable).Value
   )
   TriggerAnyTriggerAny3_3 = "TriggerAny.Trigger Any (3)"(In 2=@n39.Value, In 0=@n35.Value, In 1=@n37.Value)
   n41 = InitVariable.Trigger()
   TriggerControl_3 = UE.Trigger Control(
    Close=@TriggerAnyTriggerAny3_3.Out
    Trigger In=VariableAccessor.Trigger(Variable=VariableMutator.Trigger(Variable=@n41.Variable, Value=@WavePlayer_1.On Finished).Variable).Value
    Open=@n30.Value
   )
   TriggerAnyTriggerAny4_0 = "TriggerAny.Trigger Any (4)"(
    In 1=@TriggerControl_0.Trigger Out
    In 3=@TriggerControl_1.Trigger Out
    In 0=@TriggerControl_2.Trigger Out
    In 2=@TriggerControl_3.Trigger Out
   )
  ]
  links [{dst="@Out.L", src=Add_0.Out}, {dst="@Out.R", src=Add_4.Out}, {dst="@OnFinished", src=TriggerAnyTriggerAny4_0.Out}]
 }
 Mobile {
  nodes [
   InitVariable.Trigger()
   InitVariable.Trigger()
   n1 = UE.RandomFloat(Next=^OnGunshot, Max=^Random.Volume.Max, Min=^Random.Volume.Min)
   n2 = UE.RandomFloat(Next=^OnGunshot, Max=^Random.Pitch, Min=^Random.Pitch)
   FrequencyMultipliertoSemitone_0 = UE.Frequency Multiplier to Semitone:Float(Frequency Multiplier=@n2.Value)
   WavePlayer_0 = UE.Wave Player:Stereo(Pitch Shift=@FrequencyMultipliertoSemitone_0.Semitones, Play=@n2.On Next) {also {version [1,1]}}
   Multiply_0 = UE.Multiply:Audio by Float(AdditionalOperands=@n1.Value, PrimaryOperand=@WavePlayer_0.Out Left)
   Multiply_1 = UE.Multiply:Audio by Float(AdditionalOperands=@n1.Value, PrimaryOperand=@WavePlayer_0.Out Right)
  ]
  links [{dst="@Out.L", src=Multiply_0.Out}, {dst="@Out.R", src=Multiply_1.Out}, {dst="@OnFinished", src=WavePlayer_0.On Finished}]
 }
]
