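# Stereo MetaSound Source for ColdFern first-person weapon fire, stepping through interleaved PPK, Revolver, Uzi and 1911 shots from WeaponIndex.
# Default page throttles main shots with a randomly timed Trigger Once reset, while every gunshot plays a random transient across four round-robin players.
# A modulated low-ammo layer joins the mix, and completion follows a delay gated against new shots.
# Mobile page drops transients and throttling, using fewer waves and the 3p stereo patch.
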MetaSoundSource /WildEstateWeapons/Gameplay/ColdFern/Sound/MSS_ColdFern_Fire_1P
inputs [
 bIsFiring {type=Bool}
 BurstRate {type=Time, access=Constructor}
 Character.Affiliation {type=Enum:FortAudioAffiliation}
 Character.Health {type=Float}
 Character.UniqueID {type=Int32}
 FireBurst {type=Trigger}
 FireSingleShot {type=Trigger}
 IndividualWeaponFireVolume {type=Float}
 LowAmmo.ADSMainVolumeMultiply {type=Float}
 LowAmmo.Random.MaxPitch {type=Float}
 LowAmmo.Random.MaxVolume {type=Float}
 LowAmmo.Random.MinPitch {type=Float}
 LowAmmo.Random.MinVolume {type=Float}
 LowAmmo.Threshold {type=Int32}
 MainFire.Random.MaxPitch {type=Float}
 MainFire.Random.MaxVolume {type=Float}
 MainFire.Random.MinPitch {type=Float}
 MainFire.Random.MinVolume {type=Float}
 MainFire.Sounds {type=WaveAsset:Array}
 Mixing.LowAmmoVolume {type=Float}
 Mixing.MainFireVolume {type=Float}
 Mixing.OutputVolume {type=Float}
 RetriggerDelayTime {type=Time}
 RoundCount {type=Int32}
 RoundsPerBurst {type=Int32}
 StopBurst {type=Trigger}
 Transient Level {type=Float}
 Transients {type=WaveAsset:Array}
 UE.Attenuation.Distance {type=Float}
 UE.Source.OnPlay {type=Trigger}
 Weapon.AmmoRemaining {type=Int32}
 Weapon.ClipSize {type=Int32}
 Weapon.FireRate {type=Float}
 Weapon.FireStartTime {type=Float}
 Weapon.IsAimingDownSights {type=Bool}
 Weapon.IsSuppressorEnabled {type=Bool}
 WeaponIndex {type=Int32}
]
interfaces [Fort.Character:1.0, Fort.Weapon:1.0, UE.Attenuation:1.0, UE.OutputFormat.Stereo:1.0, UE.Source.OneShot:1.0, UE.Source:1.1]
outputs [
 bIsAlmostFinished {type=Bool}
 "UE.OutputFormat.Stereo.Audio:0" {type=Audio}
 "UE.OutputFormat.Stereo.Audio:1" {type=Audio}
 UE.Source.OneShot.OnFinished {type=Trigger}
]
pages [
 Default {
  is_build_graph=true
  input_defaults {
   BurstRate=0
   Character.Affiliation=0
   Character.Health=0
   Character.UniqueID=0
   FireBurst=null
   FireSingleShot=null
   IndividualWeaponFireVolume=1.544
   LowAmmo.ADSMainVolumeMultiply=1
   LowAmmo.Random.MaxPitch=1.1
   LowAmmo.Random.MaxVolume=1.3
   LowAmmo.Random.MinPitch=0.95
   LowAmmo.Random.MinVolume=1
   LowAmmo.Threshold=10
   MainFire.Random.MaxPitch=1.5
   MainFire.Random.MaxVolume=1.3
   MainFire.Random.MinPitch=0.75
   MainFire.Random.MinVolume=0.8
   Mixing.LowAmmoVolume=0.7256
   Mixing.MainFireVolume=1
   Mixing.OutputVolume=0.352
   RetriggerDelayTime=2
   RoundCount=24
   RoundsPerBurst=1
   StopBurst=null
   Transient Level=1
   UE.Attenuation.Distance=0
   UE.Source.OnPlay=null
   Weapon.AmmoRemaining=40
   Weapon.ClipSize=30
   Weapon.FireRate=0.0781
   Weapon.FireStartTime=0
   Weapon.IsAimingDownSights=null
   Weapon.IsSuppressorEnabled=null
   WeaponIndex=0
   bIsFiring="<MetaSoundBuilderResult.SUCCEEDED: 0>"
   MainFire.Sounds [
    /WildEstateWeapons/Gameplay/ColdFern/Sound/wav/ColdFern_Wpn_PPK_Fire_1P_01
    /WildEstateWeapons/Gameplay/ColdFern/Sound/wav/ColdFern_Wpn_Revolver_Fire_1P_01
    /WildEstateWeapons/Gameplay/ColdFern/Sound/wav/ColdFern_Wpn_Uzi_Fire_1P_01
    /WildEstateWeapons/Gameplay/ColdFern/Sound/wav/ColdFern_Wpn_1911_Fire_1P_01
    /WildEstateWeapons/Gameplay/ColdFern/Sound/wav/ColdFern_Wpn_PPK_Fire_1P_02
    /WildEstateWeapons/Gameplay/ColdFern/Sound/wav/ColdFern_Wpn_Revolver_Fire_1P_02
    /WildEstateWeapons/Gameplay/ColdFern/Sound/wav/ColdFern_Wpn_Uzi_Fire_1P_02
    /WildEstateWeapons/Gameplay/ColdFern/Sound/wav/ColdFern_Wpn_1911_Fire_1P_02
    /WildEstateWeapons/Gameplay/ColdFern/Sound/wav/ColdFern_Wpn_PPK_Fire_1P_03
    /WildEstateWeapons/Gameplay/ColdFern/Sound/wav/ColdFern_Wpn_Revolver_Fire_1P_03
    /WildEstateWeapons/Gameplay/ColdFern/Sound/wav/ColdFern_Wpn_Uzi_Fire_1P_03
    /WildEstateWeapons/Gameplay/ColdFern/Sound/wav/ColdFern_Wpn_1911_Fire_1P_03
    /WildEstateWeapons/Gameplay/ColdFern/Sound/wav/ColdFern_Wpn_PPK_Fire_1P_04
    /WildEstateWeapons/Gameplay/ColdFern/Sound/wav/ColdFern_Wpn_Revolver_Fire_1P_04
    /WildEstateWeapons/Gameplay/ColdFern/Sound/wav/ColdFern_Wpn_Uzi_Fire_1P_04
    /WildEstateWeapons/Gameplay/ColdFern/Sound/wav/ColdFern_Wpn_1911_Fire_1P_04
    /WildEstateWeapons/Gameplay/ColdFern/Sound/wav/ColdFern_Wpn_PPK_Fire_1P_05
    /WildEstateWeapons/Gameplay/ColdFern/Sound/wav/ColdFern_Wpn_Revolver_Fire_1P_05
    /WildEstateWeapons/Gameplay/ColdFern/Sound/wav/ColdFern_Wpn_Uzi_Fire_1P_05
    /WildEstateWeapons/Gameplay/ColdFern/Sound/wav/ColdFern_Wpn_1911_Fire_1P_05
    /WildEstateWeapons/Gameplay/ColdFern/Sound/wav/ColdFern_Wpn_PPK_Fire_1P_06
    /WildEstateWeapons/Gameplay/ColdFern/Sound/wav/ColdFern_Wpn_Revolver_Fire_1P_06
    /WildEstateWeapons/Gameplay/ColdFern/Sound/wav/ColdFern_Wpn_Uzi_Fire_1P_06
    /WildEstateWeapons/Gameplay/ColdFern/Sound/wav/ColdFern_Wpn_1911_Fire_1P_06
   ]
   Transients [
    /WildEstateWeapons/Gameplay/ColdFern/Sound/wav/ColdFern_Wpn_1P_Transient_01
    /WildEstateWeapons/Gameplay/ColdFern/Sound/wav/ColdFern_Wpn_1P_Transient_02
    /WildEstateWeapons/Gameplay/ColdFern/Sound/wav/ColdFern_Wpn_1P_Transient_03
    /WildEstateWeapons/Gameplay/ColdFern/Sound/wav/ColdFern_Wpn_1P_Transient_04
    /WildEstateWeapons/Gameplay/ColdFern/Sound/wav/ColdFern_Wpn_1P_Transient_05
    /WildEstateWeapons/Gameplay/ColdFern/Sound/wav/ColdFern_Wpn_1P_Transient_06
    /WildEstateWeapons/Gameplay/ColdFern/Sound/wav/ColdFern_Wpn_1P_Transient_07
    /WildEstateWeapons/Gameplay/ColdFern/Sound/wav/ColdFern_Wpn_1P_Transient_08
   ]
  }
  nodes [
   InitVariable.Enum:SubdivisionQuantizationType()
   InitVariable.Enum:SubdivisionQuantizationType()
   InitVariable.Trigger()
   n1 = InitVariable.Trigger()
   n2 = /FortMetaSoundWeapon/Items/Metasounds/Subgraphs/MSP_Weapon_Fire_Control_1p(
    BurstRate=^BurstRate
    FireBurst=^FireBurst
    FireSingleShot=^FireSingleShot
    RoundsPerBurst=^RoundsPerBurst
    StopBurst=^StopBurst
    OnPlay=^UE.Source.OnPlay
    Weapon.AmmoRemaining=^Weapon.AmmoRemaining
    Weapon.FireRate=^Weapon.FireRate
    bIsFiring=^bIsFiring
   )
   n3 = VariableDeferredAccessor.Trigger(Variable=InitVariable.Trigger().Variable)
   TriggerOnce_0 = UE.Trigger Once(Trigger In=@n2.OutGunshot, Reset=@n3.Value)
   TriggerCounter_0 = UE.Trigger Counter(
    Reset Count=^RoundCount
    Start Value=Convert.Float:Int32(Int32=^WeaponIndex).Float
    In=VariableAccessor.Trigger(Variable=VariableMutator.Trigger(Variable=@n1.Variable, Value=@TriggerOnce_0.Trigger Out).Variable).Value
   )
   n4 = UE.RandomFloat(Max=^MainFire.Random.MaxPitch, Min=^MainFire.Random.MinPitch, Next=@TriggerCounter_0.On Trigger)
   n5 = /WildEstateWeapons/Gameplay/ColdFern/Sound/Patches/MSP_ColdFern_Weapon_Fire_1p(
    Random.Volume.Max=^MainFire.Random.MaxVolume
    Random.Volume.Min=^MainFire.Random.MinVolume
    Sound=Array.Get:WaveAsset:Array(Array=^MainFire.Sounds, Index=@TriggerCounter_0.Count, Trigger=@TriggerCounter_0.On Trigger).Element
    OnGunshot=@n4.On Next
    Random.Pitch=@n4.Value
   )
   n6 = VariableAccessor.Trigger(Variable=VariableMutator.Trigger(Variable=InitVariable.Trigger().Variable, Value=@n2.OutGunshot).Variable)
   n7 = VariableAccessor.Trigger(Variable=@n6.Variable)
   ArrayRandomGet_0 = Array.Random Get:WaveAsset:Array(In Array=^Transients, Next=VariableAccessor.Trigger(Variable=@n7.Variable).Value) {also {defaults {Weights []}}}
   n8 = UE.Reroute(Value=UE.Reroute(Value=@ArrayRandomGet_0.Value).Value)
   TriggerSequence4_0 = "UE.Trigger Sequence (4)"(In=@ArrayRandomGet_0.On Next)
   WavePlayer_0 = UE.Wave Player:Stereo(Wave Asset=@n8.Value, Play=@TriggerSequence4_0.Out 1) {also {version [1,1]}}
   WavePlayer_1 = UE.Wave Player:Stereo(Wave Asset=UE.Reroute(Value=@n8.Value).Value, Play=@TriggerSequence4_0.Out 0) {also {version [1,1]}}
   WavePlayer_2 = UE.Wave Player:Stereo(Wave Asset=@n8.Value, Play=@TriggerSequence4_0.Out 3) {also {version [1,1]}}
   WavePlayer_3 = UE.Wave Player:Stereo(Wave Asset=@n8.Value, Play=@TriggerSequence4_0.Out 2) {also {version [1,1]}}
   AudioMixerAudioMixerStereo5_0 = "AudioMixer.Audio Mixer (Stereo, 5)"(
    Gain 4=^IndividualWeaponFireVolume
    Gain 0=^Transient Level
    Gain 1=^Transient Level
    Gain 2=^Transient Level
    Gain 3=^Transient Level
    In 4 L=@n5.Out.L
    In 4 R=@n5.Out.R
    In 1 L=@WavePlayer_0.Out Left
    In 1 R=@WavePlayer_0.Out Right
    In 0 L=@WavePlayer_1.Out Left
    In 0 R=@WavePlayer_1.Out Right
    In 3 L=@WavePlayer_2.Out Left
    In 3 R=@WavePlayer_2.Out Right
    In 2 L=@WavePlayer_3.Out Left
    In 2 R=@WavePlayer_3.Out Right
   )
   n9 = InitVariable.Int32()
   n10 = /FortMetaSoundWeapon/Items/Metasounds/Subgraphs/MSP_Weapon_Fire_LowAmmo(
    LowAmmo.ADS.VolumeMultiply=^LowAmmo.ADSMainVolumeMultiply
    LowAmmo.Random.Pitch.Max=^LowAmmo.Random.MaxPitch
    LowAmmo.Random.Volume.Max=^LowAmmo.Random.MaxVolume
    LowAmmo.Random.Pitch.Min=^LowAmmo.Random.MinPitch
    LowAmmo.Random.Volume.Min=^LowAmmo.Random.MinVolume
    LowAmmo.Threshold=^LowAmmo.Threshold
    Weapon.IsAimingDownSights=^Weapon.IsAimingDownSights
    CurrentAmmo=VariableAccessor.Int32(Variable=VariableMutator.Int32(Variable=@n9.Variable, Value=@n2.CurrentAmmo).Variable).Value
    OnGunshot=@n7.Value
    LowAmmo.Sounds=VariableAccessor.WaveAsset:Array(Variable=InitVariable.WaveAsset:Array().Variable).Value
   )
   n11 = Modulation.GetModulatorValue() {also {defaults {Modulator=/FortMetaSoundWeapon/Items/Modulation/CB_WeaponFire_1p_LowAmmo}}}
   AudioMixerAudioMixerStereo2_0 = "AudioMixer.Audio Mixer (Stereo, 2)"(
    Gain 0=^Mixing.MainFireVolume
    In 0 L=@AudioMixerAudioMixerStereo5_0.Out L
    In 0 R=@AudioMixerAudioMixerStereo5_0.Out R
    In 1 L=@n10.LowAmmo.C
    In 1 R=@n10.LowAmmo.C
    Gain 1=UE.Reroute(Value=UE.Multiply:Float(PrimaryOperand=^Mixing.LowAmmoVolume, AdditionalOperands=@n11.Out).Out).Value
   )
   Multiply_0 = UE.Multiply:Audio by Float(AdditionalOperands=^Mixing.OutputVolume, PrimaryOperand=@AudioMixerAudioMixerStereo2_0.Out R)
   Multiply_1 = UE.Multiply:Audio by Float(AdditionalOperands=^Mixing.OutputVolume, PrimaryOperand=@AudioMixerAudioMixerStereo2_0.Out L)
   n12 = InitVariable.Trigger()
   n13 = InitVariable.Trigger()
   TriggerAccumulatorTriggerAccumulate2_0 = "TriggerAccumulator.Trigger Accumulate (2)"(
    In 1=VariableAccessor.Trigger(Variable=VariableMutator.Trigger(Variable=@n12.Variable, Value=@n10.OnFinished).Variable).Value
    In 0=VariableAccessor.Trigger(Variable=VariableMutator.Trigger(Variable=@n13.Variable, Value=@n5.OnFinished).Variable).Value
   ) {also {defaults {Auto Reset="<MetaSoundBuilderResult.SUCCEEDED: 0>"}}}
   n14 = &MSP_TriggerResolve(A=UE.Reroute(Value=@TriggerAccumulatorTriggerAccumulate2_0.Out).Value, B=@n6.Value)
   TriggerDelay_1 = UE.Trigger Delay(Delay Time=^RetriggerDelayTime, In=@TriggerAccumulatorTriggerAccumulate2_0.Out)
   TriggerControl_0 = UE.Trigger Control(Open=@n14.A, Close=@n14.B, Trigger In=@TriggerDelay_1.Out)
   ValueValue_0 = Value.Value:Bool(Set=@TriggerControl_0.Trigger Out)
   n15 = UE.RandomTime(Next=@TriggerOnce_0.Trigger Out) {also {defaults {Max=0.29, Min=0.1}}}
   TriggerDelay_0 = UE.Trigger Delay(In=@n15.On Next, Delay Time=@n15.Value)
   VariableMutator.Trigger(Value=@TriggerDelay_0.Out, Variable=@n3.Variable)
  ]
  links [
   {dst="@UE.OutputFormat.Stereo.Audio:1", src=Multiply_0.Out}
   {dst="@UE.OutputFormat.Stereo.Audio:0", src=Multiply_1.Out}
   {dst="@UE.Source.OneShot.OnFinished", src=ValueValue_0.On Set}
   {dst="@bIsAlmostFinished", src=ValueValue_0.Output Value}
  ]
 }
 Mobile {
  input_defaults {
   RoundCount=8
   MainFire.Sounds [
    &ColdFern_Wpn_PPK_Fire_1P_01
    &ColdFern_Wpn_Revolver_Fire_1P_01
    &ColdFern_Wpn_Uzi_Fire_1P_01
    &ColdFern_Wpn_1911_Fire_1P_01
    &ColdFern_Wpn_PPK_Fire_1P_02
    &ColdFern_Wpn_Revolver_Fire_1P_02
    &ColdFern_Wpn_Uzi_Fire_1P_02
    &ColdFern_Wpn_1911_Fire_1P_02
   ]
  }
  nodes [
   InitVariable.Enum:SubdivisionQuantizationType()
   InitVariable.Enum:SubdivisionQuantizationType()
   InitVariable.Trigger()
   n1 = &MSP_Weapon_Fire_Control_1p(
    BurstRate=^BurstRate
    FireBurst=^FireBurst
    FireSingleShot=^FireSingleShot
    RoundsPerBurst=^RoundsPerBurst
    StopBurst=^StopBurst
    OnPlay=^UE.Source.OnPlay
    Weapon.AmmoRemaining=^Weapon.AmmoRemaining
    Weapon.FireRate=^Weapon.FireRate
    bIsFiring=^bIsFiring
   )
   n2 = VariableAccessor.Trigger(Variable=VariableMutator.Trigger(Variable=InitVariable.Trigger().Variable, Value=@n1.OutGunshot).Variable)
   n3 = VariableAccessor.Trigger(Variable=@n2.Variable)
   TriggerCounter_0 = UE.Trigger Counter(Reset Count=^RoundCount, Start Value=Convert.Float:Int32(Int32=^WeaponIndex).Float, In=VariableAccessor.Trigger(Variable=@n3.Variable).Value)
   n4 = UE.RandomFloat(Max=^MainFire.Random.MaxPitch, Min=^MainFire.Random.MinPitch, Next=@TriggerCounter_0.On Trigger)
   n5 = /WildEstateWeapons/Gameplay/ColdFern/Sound/Patches/MSP_ColdFern_Weapon_Fire_3p_Stereo(
    Random.Volume.Max=^MainFire.Random.MaxVolume
    Random.Volume.Min=^MainFire.Random.MinVolume
    Sound=Array.Get:WaveAsset:Array(Array=^MainFire.Sounds, Index=@TriggerCounter_0.Count, Trigger=@TriggerCounter_0.On Trigger).Element
    OnGunshot=@n4.On Next
    Random.Pitch=@n4.Value
   )
   n6 = InitVariable.Int32()
   n7 = &MSP_Weapon_Fire_LowAmmo(
    LowAmmo.ADS.VolumeMultiply=^LowAmmo.ADSMainVolumeMultiply
    LowAmmo.Random.Pitch.Max=^LowAmmo.Random.MaxPitch
    LowAmmo.Random.Volume.Max=^LowAmmo.Random.MaxVolume
    LowAmmo.Random.Pitch.Min=^LowAmmo.Random.MinPitch
    LowAmmo.Random.Volume.Min=^LowAmmo.Random.MinVolume
    LowAmmo.Threshold=^LowAmmo.Threshold
    Weapon.IsAimingDownSights=^Weapon.IsAimingDownSights
    CurrentAmmo=VariableAccessor.Int32(Variable=VariableMutator.Int32(Variable=@n6.Variable, Value=@n1.CurrentAmmo).Variable).Value
    OnGunshot=@n3.Value
    LowAmmo.Sounds=VariableAccessor.WaveAsset:Array(Variable=InitVariable.WaveAsset:Array().Variable).Value
   )
   AudioMixerAudioMixerStereo2_0 = "AudioMixer.Audio Mixer (Stereo, 2)"(
    Gain 0=^Mixing.MainFireVolume
    In 0 L=@n5.Out.L
    In 0 R=@n5.Out.R
    In 1 L=@n7.LowAmmo.C
    In 1 R=@n7.LowAmmo.C
    Gain 1=UE.Multiply:Float(
     PrimaryOperand=^Mixing.LowAmmoVolume
     AdditionalOperands=Modulation.GetModulatorValue() {also {defaults {Modulator=&CB_WeaponFire_1p_LowAmmo}}}.Out
    ).Out
   )
   Multiply_0 = UE.Multiply:Audio by Float(AdditionalOperands=^Mixing.OutputVolume, PrimaryOperand=@AudioMixerAudioMixerStereo2_0.Out R)
   Multiply_1 = UE.Multiply:Audio by Float(AdditionalOperands=^Mixing.OutputVolume, PrimaryOperand=@AudioMixerAudioMixerStereo2_0.Out L)
   n8 = InitVariable.Trigger()
   n9 = InitVariable.Trigger()
   TriggerAccumulatorTriggerAccumulate2_0 = "TriggerAccumulator.Trigger Accumulate (2)"(
    In 1=VariableAccessor.Trigger(Variable=VariableMutator.Trigger(Variable=@n8.Variable, Value=@n7.OnFinished).Variable).Value
    In 0=VariableAccessor.Trigger(Variable=VariableMutator.Trigger(Variable=@n9.Variable, Value=@n5.OnFinished).Variable).Value
   ) {also {defaults {Auto Reset="<MetaSoundBuilderResult.SUCCEEDED: 0>"}}}
   n10 = &MSP_TriggerResolve(A=UE.Reroute(Value=@TriggerAccumulatorTriggerAccumulate2_0.Out).Value, B=@n2.Value)
   TriggerDelay_0 = UE.Trigger Delay(Delay Time=^RetriggerDelayTime, In=@TriggerAccumulatorTriggerAccumulate2_0.Out)
   TriggerControl_0 = UE.Trigger Control(Open=@n10.A, Close=@n10.B, Trigger In=@TriggerDelay_0.Out)
   ValueValue_0 = Value.Value:Bool(Set=@TriggerControl_0.Trigger Out)
  ]
  links [
   {dst="@UE.OutputFormat.Stereo.Audio:1", src=Multiply_0.Out}
   {dst="@UE.OutputFormat.Stereo.Audio:0", src=Multiply_1.Out}
   {dst="@UE.Source.OneShot.OnFinished", src=ValueValue_0.On Set}
   {dst="@bIsAlmostFinished", src=ValueValue_0.Output Value}
  ]
 }
]
properties {
 assetImportData=&self:AssetImportData
 numChannels=2
 outputFormat=Stereo
 sampleRate=48000
 soundClassObject=&Fort_WeaponRanged_Fire_1P_PSM
 concurrencySet [&Weapon_Gunfire_Local_Conc]
 modulationSettings {
  volumeRouting=Override
  volumeModulationDestination {bEnableModulation=true, modulators [/WildEstateWeapons/Gameplay/ColdFern/Sound/Modulation/MP_ColdFernEnvFollower]}
 }
}
