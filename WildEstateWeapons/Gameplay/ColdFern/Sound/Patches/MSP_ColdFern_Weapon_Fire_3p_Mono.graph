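# MetaSound patch that plays third-person ColdFern weapon gunshots in mono, picking the wave from an array by index with a pitch multiplier converted to semitones.
# A trigger sequence rotates shots across three wave players, or bUseSingleVoice keeps one; each gets random volume before summing.
# Trigger gates pass OnFinished only from the most recently started player.
# A Mobile page uses one randomized player with no wave asset wired.
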MetaSoundPatch /WildEstateWeapons/Gameplay/ColdFern/Sound/Patches/MSP_ColdFern_Weapon_Fire_3p_Mono
inputs [
 bUseSingleVoice {type=Bool}
 OnGunshot {type=Trigger}
 Random.Pitch {type=Float}
 Random.Volume.Max {type=Float}
 Random.Volume.Min {type=Float}
 Sound {type=WaveAsset:Array}
 SoundIndex {type=Int32}
]
outputs [
 OnFinished {type=Trigger}
 Out.Out {type=Audio}
]
pages [
 Default {
  is_build_graph=true
  input_defaults {
   OnGunshot=null
   Random.Pitch=0
   Random.Volume.Max=1
   Random.Volume.Min=0
   SoundIndex=0
   bUseSingleVoice="<MetaSoundBuilderResult.SUCCEEDED: 0>"
   Sound []
  }
  nodes [
   n1 = TriggerCompare.Clamp:Bool(Compare=^OnGunshot, A=^bUseSingleVoice) {also {defaults {B="<MetaSoundBuilderResult.SUCCEEDED: 0>"}}}
   TriggerSequence3_0 = "UE.Trigger Sequence (3)"(In=@n1.False)
   TriggerAnyTriggerAny2_2 = "TriggerAny.Trigger Any (2)"(In 0=@n1.True, In 1=@TriggerSequence3_0.Out 0)
   n2 = VariableAccessor.Trigger(Variable=VariableMutator.Trigger(Variable=InitVariable.Trigger().Variable, Value=@TriggerAnyTriggerAny2_2.Out).Variable)
   n3 = UE.RandomFloat(Max=^Random.Volume.Max, Min=^Random.Volume.Min, Next=@n2.Value)
   n4 = Array.Get:WaveAsset:Array(Trigger=^OnGunshot, Array=^Sound, Index=^SoundIndex)
   FrequencyMultipliertoSemitone_0 = UE.Frequency Multiplier to Semitone:Float(Frequency Multiplier=^Random.Pitch)
   n5 = VariableAccessor.Float(Variable=VariableMutator.Float(Value=@FrequencyMultipliertoSemitone_0.Semitones, Variable=InitVariable.Float().Variable).Variable)
   n6 = VariableAccessor.Float(Variable=@n5.Variable)
   WavePlayer_1 = UE.Wave Player:Mono(Wave Asset=@n4.Element, Pitch Shift=VariableAccessor.Float(Variable=@n6.Variable).Value, Play=@n2.Value) {also {version [1,1]}}
   n7 = VariableAccessor.Trigger(Variable=VariableMutator.Trigger(Variable=InitVariable.Trigger().Variable, Value=@TriggerSequence3_0.Out 1).Variable)
   n8 = UE.RandomFloat(Max=^Random.Volume.Max, Min=^Random.Volume.Min, Next=@n7.Value)
   WavePlayer_2 = UE.Wave Player:Mono(Wave Asset=UE.Reroute(Value=@n4.Element).Value, Pitch Shift=@n6.Value, Play=@n7.Value) {also {version [1,1]}}
   n9 = VariableAccessor.Trigger(Variable=VariableMutator.Trigger(Variable=InitVariable.Trigger().Variable, Value=@TriggerSequence3_0.Out 2).Variable)
   WavePlayer_0 = UE.Wave Player:Mono(Wave Asset=UE.Reroute(Value=@n4.Element).Value, Pitch Shift=@n5.Value, Play=@n9.Value) {also {version [1,1]}}
   Multiply_0 = UE.Multiply:Audio by Float(
    AdditionalOperands=UE.RandomFloat(Max=^Random.Volume.Max, Min=^Random.Volume.Min, Next=@n9.Value).Value
    PrimaryOperand=@WavePlayer_0.Out Mono
   )
   Add_0 = UE.Add:Audio(
    PrimaryOperand=UE.Add:Audio(
     PrimaryOperand=UE.Multiply:Audio by Float(AdditionalOperands=@n3.Value, PrimaryOperand=@WavePlayer_1.Out Mono).Out
     AdditionalOperands=UE.Multiply:Audio by Float(AdditionalOperands=@n8.Value, PrimaryOperand=@WavePlayer_2.Out Mono).Out
    ).Out
    AdditionalOperands=@Multiply_0.Out
   )
   n10 = VariableAccessor.Trigger(Variable=VariableMutator.Trigger(Variable=InitVariable.Trigger().Variable, Value=@WavePlayer_1.On Play).Variable)
   n11 = VariableAccessor.Trigger(Variable=@n10.Variable)
   n12 = VariableAccessor.Trigger(Variable=VariableMutator.Trigger(Variable=InitVariable.Trigger().Variable, Value=@WavePlayer_2.On Play).Variable)
   n13 = VariableAccessor.Trigger(Variable=@n12.Variable)
   TriggerAnyTriggerAny2_1 = "TriggerAny.Trigger Any (2)"(In 0=VariableAccessor.Trigger(Variable=@n11.Variable).Value, In 1=VariableAccessor.Trigger(Variable=@n13.Variable).Value)
   n14 = InitVariable.Trigger()
   n15 = VariableAccessor.Trigger(Variable=VariableMutator.Trigger(Variable=InitVariable.Trigger().Variable, Value=@WavePlayer_0.On Play).Variable)
   n16 = VariableAccessor.Trigger(Variable=@n15.Variable)
   TriggerControl_0 = UE.Trigger Control(
    Close=@TriggerAnyTriggerAny2_1.Out
    Trigger In=VariableAccessor.Trigger(Variable=VariableMutator.Trigger(Variable=@n14.Variable, Value=@WavePlayer_0.On Finished).Variable).Value
    Open=@n16.Value
   )
   TriggerAnyTriggerAny2_0 = "TriggerAny.Trigger Any (2)"(In 0=@n12.Value, In 1=@n15.Value)
   n17 = InitVariable.Trigger()
   TriggerControl_1 = UE.Trigger Control(
    Close=@TriggerAnyTriggerAny2_0.Out
    Open=@n11.Value
    Trigger In=VariableAccessor.Trigger(Variable=VariableMutator.Trigger(Variable=@n17.Variable, Value=@WavePlayer_1.On Finished).Variable).Value
   )
   TriggerAnyTriggerAny2_3 = "TriggerAny.Trigger Any (2)"(In 0=@n10.Value, In 1=VariableAccessor.Trigger(Variable=@n16.Variable).Value)
   n18 = InitVariable.Trigger()
   TriggerControl_2 = UE.Trigger Control(
    Close=@TriggerAnyTriggerAny2_3.Out
    Trigger In=VariableAccessor.Trigger(Variable=VariableMutator.Trigger(Variable=@n18.Variable, Value=@WavePlayer_2.On Finished).Variable).Value
    Open=@n13.Value
   )
   TriggerAnyTriggerAny3_0 = "TriggerAny.Trigger Any (3)"(In 2=@TriggerControl_0.Trigger Out, In 0=@TriggerControl_1.Trigger Out, In 1=@TriggerControl_2.Trigger Out)
  ]
  links [{dst="@Out.Out", src=Add_0.Out}, {dst="@OnFinished", src=TriggerAnyTriggerAny3_0.Out}]
 }
 Mobile {
  nodes [
   InitVariable.Trigger()
   InitVariable.Trigger()
   n1 = UE.RandomFloat(Next=^OnGunshot, Min=^Random.Pitch)
   FrequencyMultipliertoSemitone_0 = UE.Frequency Multiplier to Semitone:Float(Frequency Multiplier=@n1.Value)
   WavePlayer_0 = UE.Wave Player:Mono(Pitch Shift=@FrequencyMultipliertoSemitone_0.Semitones, Play=@n1.On Next) {also {version [1,1]}}
   Multiply_0 = UE.Multiply:Audio by Float(
    AdditionalOperands=UE.RandomFloat(Next=^OnGunshot, Max=^Random.Volume.Max, Min=^Random.Volume.Min).Value
    PrimaryOperand=@WavePlayer_0.Out Mono
   )
  ]
  links [{dst="@Out.Out", src=Multiply_0.Out}, {dst="@OnFinished", src=WavePlayer_0.On Finished}]
 }
]
