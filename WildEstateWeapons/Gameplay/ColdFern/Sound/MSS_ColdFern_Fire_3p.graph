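# Stereo MetaSound source for ColdFern's third-person gunfire, mixing close and distant sample sets.
# Fire Control emits gunshots; a counter indexes interleaved 1911, PPK, revolver and Uzi samples, and distance curves gate the stereo close and mono distant players.
# A character-ID profile randomizes volume and EQ before shared 3p mixing, distance-driven high/low-pass filters and master volume.
# Finish fires after both layers end plus a retrigger delay, gated against new shots.
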MetaSoundSource /WildEstateWeapons/Gameplay/ColdFern/Sound/MSS_ColdFern_Fire_3p
inputs [
 bEnableDistanceEQ {type=Bool}
 bHasUnlimitedAmmo {type=Bool}
 bIsAlwaysSuppressed {type=Bool}
 BurstRate {type=Time}
 bUseSingleVoice {type=Bool}
 Character.Affiliation {type=Enum:FortAudioAffiliation}
 Character.Health {type=Float}
 Character.UniqueID {type=Int32}
 DistanceCurves {type=WaveTableBankAsset}
 Envelope.A.SustainTime {type=Time}
 FireBurst {type=Trigger}
 FireSingleShot {type=Trigger}
 MainFire.DistanceCurves {type=WaveTableBankAsset}
 MainFire.FalloffDistance {type=Float}
 MainFire.SoundsClose {type=WaveAsset:Array}
 MainFire.SoundsDistant {type=WaveAsset:Array}
 MainFire.SoundsMedium {type=WaveAsset:Array}
 MasterVolume {type=Float}
 Mixing.FalloffDistance {type=Float}
 Mixing.Friendly.FriendlyFireWeaponVolume {type=Float}
 Mixing.Volume.Close {type=Float}
 Mixing.Volume.Distant {type=Float}
 Random.EQ.CutoffFrequency.Max {type=Float}
 Random.EQ.CutoffFrequency.Min {type=Float}
 Random.EQ.Gain.Max {type=Float}
 Random.EQ.Gain.Min {type=Float}
 Random.Pitch.Max {type=Float}
 Random.Pitch.Min {type=Float}
 Random.Volume.Max {type=Float}
 Random.Volume.Min {type=Float}
 Release Time {type=Time}
 RetriggerDelayTime {type=Time}
 RoundAmmount {type=Int32}
 RoundsPerBurst {type=Int32}
 StartFiring {type=Trigger}
 StopBurst {type=Trigger}
 StopFiring {type=Trigger}
 Suppressor.Sounds.Close {type=WaveAsset:Array}
 Suppressor.Sounds.Medium {type=WaveAsset:Array}
 Suppressor.Volume.Close {type=Float}
 Suppressor.Volume.Medium {type=Float}
 UE.Attenuation.Distance {type=Float}
 UE.Source.OnPlay {type=Trigger}
 UE.Spatialization.Azimuth {type=Float}
 UE.Spatialization.Elevation {type=Float}
 Weapon.AmmoRemaining {type=Int32}
 Weapon.ClipSize {type=Int32}
 Weapon.FireRate {type=Float}
 Weapon.FireStartTime {type=Float}
 Weapon.IsAimingDownSights {type=Bool}
 Weapon.IsSuppressorEnabled {type=Bool}
]
interfaces [
 Fort.Character:1.0
 Fort.Weapon:1.0
 UE.Attenuation:1.0
 UE.OutputFormat.Stereo:1.0
 UE.Source.OneShot:1.0
 UE.Source:1.1
 UE.Spatialization:1.0
]
outputs [
 bIsAlmostFinished {type=Bool}
 "UE.OutputFormat.Stereo.Audio:0" {type=Audio}
 "UE.OutputFormat.Stereo.Audio:1" {type=Audio}
 UE.Source.OneShot.OnFinished {type=Trigger}
]
pages [Default {
  input_defaults {
   BurstRate=0
   Character.Affiliation=0
   Character.Health=1
   Character.UniqueID=0
   DistanceCurves=&/ColdFern_DistanceCurves
   Envelope.A.SustainTime=0.064
   FireBurst=null
   FireSingleShot=null
   MainFire.DistanceCurves=&ColdFern_DistanceCurves
   MainFire.FalloffDistance=27500
   MasterVolume=1
   Mixing.FalloffDistance=27500
   Mixing.Friendly.FriendlyFireWeaponVolume=0.7
   Mixing.Volume.Close=0.9
   Mixing.Volume.Distant=1
   Random.EQ.CutoffFrequency.Max=20
   Random.EQ.CutoffFrequency.Min=20
   Random.EQ.Gain.Max=0
   Random.EQ.Gain.Min=-3
   Random.Pitch.Max=1.3
   Random.Pitch.Min=0.5
   Random.Volume.Max=1.4
   Random.Volume.Min=0.4
   Release Time=0
   RetriggerDelayTime=2
   RoundAmmount=24
   RoundsPerBurst=2
   StartFiring=null
   StopBurst=null
   StopFiring=null
   Suppressor.Volume.Close=0.5
   Suppressor.Volume.Medium=0.3
   UE.Attenuation.Distance=19030.0098
   UE.Source.OnPlay=null
   UE.Spatialization.Azimuth=0
   UE.Spatialization.Elevation=0
   Weapon.AmmoRemaining=80
   Weapon.ClipSize=30
   Weapon.FireRate=0.078
   Weapon.FireStartTime=0
   Weapon.IsAimingDownSights=null
   Weapon.IsSuppressorEnabled=null
   bEnableDistanceEQ="<MetaSoundBuilderResult.SUCCEEDED: 0>"
   bHasUnlimitedAmmo="<MetaSoundBuilderResult.SUCCEEDED: 0>"
   bIsAlwaysSuppressed=null
   bUseSingleVoice=null
   MainFire.SoundsClose [
    /WildEstateWeapons/Gameplay/ColdFern/Sound/wav/ColdFern_Wpn_1911_Fire_3P_Close_01
    /WildEstateWeapons/Gameplay/ColdFern/Sound/wav/ColdFern_Wpn_PPK_Fire_3P_Close_01
    /WildEstateWeapons/Gameplay/ColdFern/Sound/wav/ColdFern_Wpn_Revolver_Fire_3P_Close_01
    /WildEstateWeapons/Gameplay/ColdFern/Sound/wav/ColdFern_Wpn_Uzi_Fire_3P_Close_01
    /WildEstateWeapons/Gameplay/ColdFern/Sound/wav/ColdFern_Wpn_1911_Fire_3P_Close_02
    /WildEstateWeapons/Gameplay/ColdFern/Sound/wav/ColdFern_Wpn_PPK_Fire_3P_Close_02
    /WildEstateWeapons/Gameplay/ColdFern/Sound/wav/ColdFern_Wpn_Revolver_Fire_3P_Close_02
    /WildEstateWeapons/Gameplay/ColdFern/Sound/wav/ColdFern_Wpn_Uzi_Fire_3P_Close_02
    /WildEstateWeapons/Gameplay/ColdFern/Sound/wav/ColdFern_Wpn_1911_Fire_3P_Close_03
    /WildEstateWeapons/Gameplay/ColdFern/Sound/wav/ColdFern_Wpn_PPK_Fire_3P_Close_03
    /WildEstateWeapons/Gameplay/ColdFern/Sound/wav/ColdFern_Wpn_Revolver_Fire_3P_Close_03
    /WildEstateWeapons/Gameplay/ColdFern/Sound/wav/ColdFern_Wpn_Uzi_Fire_3P_Close_03
    /WildEstateWeapons/Gameplay/ColdFern/Sound/wav/ColdFern_Wpn_1911_Fire_3P_Close_04
    /WildEstateWeapons/Gameplay/ColdFern/Sound/wav/ColdFern_Wpn_PPK_Fire_3P_Close_04
    /WildEstateWeapons/Gameplay/ColdFern/Sound/wav/ColdFern_Wpn_Revolver_Fire_3P_Close_04
    /WildEstateWeapons/Gameplay/ColdFern/Sound/wav/ColdFern_Wpn_Uzi_Fire_3P_Close_04
    /WildEstateWeapons/Gameplay/ColdFern/Sound/wav/ColdFern_Wpn_1911_Fire_3P_Close_05
    /WildEstateWeapons/Gameplay/ColdFern/Sound/wav/ColdFern_Wpn_PPK_Fire_3P_Close_05
    /WildEstateWeapons/Gameplay/ColdFern/Sound/wav/ColdFern_Wpn_Revolver_Fire_3P_Close_05
    /WildEstateWeapons/Gameplay/ColdFern/Sound/wav/ColdFern_Wpn_Uzi_Fire_3P_Close_05
    /WildEstateWeapons/Gameplay/ColdFern/Sound/wav/ColdFern_Wpn_1911_Fire_3P_Close_06
    /WildEstateWeapons/Gameplay/ColdFern/Sound/wav/ColdFern_Wpn_PPK_Fire_3P_Close_06
    /WildEstateWeapons/Gameplay/ColdFern/Sound/wav/ColdFern_Wpn_Revolver_Fire_3P_Close_06
    /WildEstateWeapons/Gameplay/ColdFern/Sound/wav/ColdFern_Wpn_Uzi_Fire_3P_Close_06
   ]
   MainFire.SoundsDistant [
    /WildEstateWeapons/Gameplay/ColdFern/Sound/wav/ColdFern_Wpn_1911_Fire_3P_Dist_01
    /WildEstateWeapons/Gameplay/ColdFern/Sound/wav/ColdFern_Wpn_PPK_Fire_3P_Dist_01
    /WildEstateWeapons/Gameplay/ColdFern/Sound/wav/ColdFern_Wpn_Revolver_Fire_3P_Dist_01
    /WildEstateWeapons/Gameplay/ColdFern/Sound/wav/ColdFern_Wpn_Uzi_Fire_3P_Dist_01
    /WildEstateWeapons/Gameplay/ColdFern/Sound/wav/ColdFern_Wpn_1911_Fire_3P_Dist_02
    /WildEstateWeapons/Gameplay/ColdFern/Sound/wav/ColdFern_Wpn_PPK_Fire_3P_Dist_02
    /WildEstateWeapons/Gameplay/ColdFern/Sound/wav/ColdFern_Wpn_Revolver_Fire_3P_Dist_02
    /WildEstateWeapons/Gameplay/ColdFern/Sound/wav/ColdFern_Wpn_Uzi_Fire_3P_Dist_02
    /WildEstateWeapons/Gameplay/ColdFern/Sound/wav/ColdFern_Wpn_1911_Fire_3P_Dist_03
    /WildEstateWeapons/Gameplay/ColdFern/Sound/wav/ColdFern_Wpn_PPK_Fire_3P_Dist_03
    /WildEstateWeapons/Gameplay/ColdFern/Sound/wav/ColdFern_Wpn_Revolver_Fire_3P_Dist_03
    /WildEstateWeapons/Gameplay/ColdFern/Sound/wav/ColdFern_Wpn_Uzi_Fire_3P_Dist_03
    /WildEstateWeapons/Gameplay/ColdFern/Sound/wav/ColdFern_Wpn_1911_Fire_3P_Dist_04
    /WildEstateWeapons/Gameplay/ColdFern/Sound/wav/ColdFern_Wpn_PPK_Fire_3P_Dist_04
    /WildEstateWeapons/Gameplay/ColdFern/Sound/wav/ColdFern_Wpn_Revolver_Fire_3P_Dist_04
    /WildEstateWeapons/Gameplay/ColdFern/Sound/wav/ColdFern_Wpn_Uzi_Fire_3P_Dist_04
    /WildEstateWeapons/Gameplay/ColdFern/Sound/wav/ColdFern_Wpn_1911_Fire_3P_Dist_05
    /WildEstateWeapons/Gameplay/ColdFern/Sound/wav/ColdFern_Wpn_PPK_Fire_3P_Dist_05
    /WildEstateWeapons/Gameplay/ColdFern/Sound/wav/ColdFern_Wpn_Revolver_Fire_3P_Dist_05
    /WildEstateWeapons/Gameplay/ColdFern/Sound/wav/ColdFern_Wpn_Uzi_Fire_3P_Dist_05
    /WildEstateWeapons/Gameplay/ColdFern/Sound/wav/ColdFern_Wpn_1911_Fire_3P_Dist_06
    /WildEstateWeapons/Gameplay/ColdFern/Sound/wav/ColdFern_Wpn_PPK_Fire_3P_Dist_06
    /WildEstateWeapons/Gameplay/ColdFern/Sound/wav/ColdFern_Wpn_Revolver_Fire_3P_Dist_06
    /WildEstateWeapons/Gameplay/ColdFern/Sound/wav/ColdFern_Wpn_Uzi_Fire_3P_Dist_06
   ]
   MainFire.SoundsMedium []
   Suppressor.Sounds.Close []
   Suppressor.Sounds.Medium []
  }
  nodes [
   n1 = /FortMetaSoundWeapon/Items/Metasounds/Subgraphs/MSP_Weapon_Fire_Control_3p(
    FireSingleShot=^FireSingleShot
    Weapon.FireRate=^Weapon.FireRate
    Weapon.AmmoRemaining=UE.Reroute(Value=^Weapon.AmmoRemaining).Value
    RoundsPerBurst=UE.Reroute(Value=^RoundsPerBurst).Value
    StartFiring=UE.Reroute(Value=^StartFiring).Value
    StopFiring=UE.Reroute(Value=^StopFiring).Value
    StopBurst=UE.Reroute(Value=^StopBurst).Value
    FireBurst=UE.Reroute(Value=^FireBurst).Value
    bHasUnlimitedAmmo=UE.Reroute(Value=^bHasUnlimitedAmmo).Value
   )
   n2 = VariableMutator.Trigger(Variable=InitVariable.Trigger().Variable, Value=@n1.OutGunshot)
   n3 = VariableAccessor.Trigger(Variable=@n2.Variable)
   TriggerCounter_0 = UE.Trigger Counter(Reset Count=^RoundAmmount, In=VariableAccessor.Trigger(Variable=@n3.Variable).Value)
   n4 = VariableAccessor.Float(Variable=VariableMutator.Float(
     Variable=InitVariable.Float().Variable
     Value=MapRange.MapRange:Float(In Range B=^MainFire.FalloffDistance, In=^UE.Attenuation.Distance).Out Value
    ).Variable)
   n5 = VariableAccessor.Float(Variable=@n4.Variable)
   n6 = TriggerCompare.Clamp:Float(
    Compare=@TriggerCounter_0.On Trigger
    A=UE.WaveTableBankEvaluate(WaveTableBank=^MainFire.DistanceCurves, Input=VariableAccessor.Float(Variable=@n5.Variable).Value).Output
   ) {also {defaults {B=0, Type=3}}}
   n7 = /FortMetaSoundWeapon/Items/Metasounds/Subgraphs/MSP_Weapon_Fire_GenerateProfile(
    Random.EQ.Frequency.Max=^Random.EQ.CutoffFrequency.Max
    Random.Volume.Min=^Random.Volume.Min
    Generate=^UE.Source.OnPlay
    Random.Volume.Max=UE.Reroute(Value=^Random.Volume.Max).Value
    Random.EQ.Frequency.Min=UE.Reroute(Value=^Random.EQ.CutoffFrequency.Min).Value
    Random.EQ.Gain.Max=UE.Reroute(Value=^Random.EQ.Gain.Max).Value
    Random.EQ.Gain.Min=UE.Reroute(Value=^Random.EQ.Gain.Min).Value
    Character.UniqueID=UE.Reroute(Value=^Character.UniqueID).Value
   ) {also {defaults {Random.Pitch.Max=1.3, Random.Pitch.Min=0.5}}}
   n8 = VariableAccessor.Float(Variable=VariableMutator.Float(Variable=InitVariable.Float().Variable, Value=@n7.Volume.Max).Variable)
   n9 = VariableAccessor.Float(Variable=VariableMutator.Float(Variable=InitVariable.Float().Variable, Value=@n7.Volume.Min).Variable)
   n10 = VariableAccessor.Float(Variable=VariableMutator.Float(Variable=InitVariable.Float().Variable) {also {defaults {Value=1}}}.Variable)
   n11 = /WildEstateWeapons/Gameplay/ColdFern/Sound/Patches/MSP_ColdFern_Weapon_Fire_3p_Stereo(
    Envelope.A.SustainTime=^Envelope.A.SustainTime
    Envelope.B.SustainTime=^Envelope.A.SustainTime
    Envelope.C.SustainTime=^Envelope.A.SustainTime
    Envelope.D.SustainTime=^Envelope.A.SustainTime
    Sound=^MainFire.SoundsClose
    bUseSingleVoice=^bUseSingleVoice
    OnGunshot=@n6.True
    SoundIndex=@TriggerCounter_0.Count
    Random.Volume.Max=@n8.Value
    Random.Volume.Min=@n9.Value
    Random.Pitch=@n10.Value
   ) {also {defaults {Envelope.A.SustainVolume=1, Envelope.B.SustainVolume=1, Envelope.C.SustainVolume=0.5, Envelope.D.SustainVolume=0.5, RingOutProbability=0}}}
   n12 = TriggerCompare.Clamp:Float(
    Compare=UE.Reroute(Value=@TriggerCounter_0.On Trigger).Value
    A=UE.WaveTableBankEvaluate(WaveTableBank=^MainFire.DistanceCurves, Input=@n5.Value) {also {defaults {Index=1}}}.Output
   ) {also {defaults {B=0, Type=3}}}
   n13 = /WildEstateWeapons/Gameplay/ColdFern/Sound/Patches/MSP_ColdFern_Weapon_Fire_3p_Mono(
    Sound=^MainFire.SoundsDistant
    bUseSingleVoice=^bUseSingleVoice
    SoundIndex=UE.Reroute(Value=@TriggerCounter_0.Count).Value
    OnGunshot=@n12.True
    Random.Volume.Max=VariableAccessor.Float(Variable=@n8.Variable).Value
    Random.Volume.Min=VariableAccessor.Float(Variable=@n9.Variable).Value
    Random.Pitch=VariableAccessor.Float(Variable=@n10.Variable).Value
   )
   n14 = InitVariable.Float()
   n15 = InitVariable.Float()
   n16 = InitVariable.Float()
   n17 = /FortMetaSoundWeapon/Items/Metasounds/Subgraphs/MSP_Weapon_Fire_Mixing_3p(
    FireSingleShot=^FireSingleShot
    Mixing.Friendly=^Mixing.Friendly.FriendlyFireWeaponVolume
    Volume.Close=^Mixing.Volume.Close
    Volume.Medium=^Mixing.Volume.Distant
    Weapon.IsAlwaysSuppressed=UE.Reroute(Value=^bIsAlwaysSuppressed).Value
    Spatialization.Elevation=UE.Reroute(Value=^UE.Spatialization.Elevation).Value
    Close.R=UE.Reroute(Value=UE.Reroute(Value=@n11.Out.R).Value).Value
    Weapon.FireRate=UE.Reroute(Value=UE.Reroute(Value=^Weapon.FireRate).Value).Value
    Orientation.Azimuth=UE.Reroute(Value=UE.Reroute(Value=^UE.Spatialization.Azimuth).Value).Value
    Close.L=UE.Reroute(Value=UE.Reroute(Value=@n11.Out.L).Value).Value
    Medium.L=UE.Reroute(Value=UE.Reroute(Value=@n13.Out.Out).Value).Value
    Character.Affiliation=UE.Reroute(Value=^Character.Affiliation).Value
    Weapon.IsSuppressorEnabled=UE.Reroute(Value=^Weapon.IsSuppressorEnabled).Value
    bEnableDistanceEQ=UE.Reroute(Value=^bEnableDistanceEQ).Value
    ProfileFilter.Gain=VariableAccessor.Float(Variable=VariableMutator.Float(Variable=@n14.Variable, Value=@n7.EQ.Gain).Variable).Value
    ProfileFilter.CutoffFrequency=VariableAccessor.Float(Variable=VariableMutator.Float(Variable=@n15.Variable, Value=@n7.EQ.CutoffFrequency).Variable).Value
    ProfileFilter.Bandwidth=VariableAccessor.Float(Variable=VariableMutator.Float(Variable=@n16.Variable, Value=@n7.EQ.Bandwidth).Variable).Value
    OnGunshot=VariableAccessor.Trigger(Variable=@n2.Variable).Value
    CurveEvaluation.Medium=UE.WaveTableBankEvaluate(WaveTableBank=^MainFire.DistanceCurves, Input=@n4.Value) {also {defaults {Index=1}}}.Output
    CurveEvaluation.Close=UE.WaveTableBankEvaluate(WaveTableBank=^MainFire.DistanceCurves, Input=@n4.Value).Output
   ) {also {defaults {Volume.Distant=0}}}
   n18 = MapRange.MapRange:Float(In Range B=^Mixing.FalloffDistance, In=^UE.Attenuation.Distance) {also {defaults {In Range A=3000, Out Range A=1, Out Range B=0.001}}}
   n19 = MapRange.MapRange:Float(In=@n18.Out Value) {also {defaults {In Range A=0.95, In Range B=0.021, Out Range B=300}}}
   OnePoleHighPassFilter_1 = UE.One-Pole High Pass Filter:Audio(In=@n17.Out.L, Cutoff Frequency=@n19.Out Value)
   n20 = UE.Reroute(Value=MapRange.MapRange:Float(In=@n18.Out Value) {also {defaults {In Range A=0.95, In Range B=0.2, Out Range A=20000, Out Range B=16000}}}.Out Value)
   OnePoleLowPassFilter_0 = UE.One-Pole Low Pass Filter:Audio(In=@OnePoleHighPassFilter_1.Out, Cutoff Frequency=@n20.Value)
   Multiply_3 = UE.Multiply:Audio by Float(AdditionalOperands=^MasterVolume, PrimaryOperand=@OnePoleLowPassFilter_0.Out)
   n21 = VariableAccessor.Float(Variable=InitVariable.Float().Variable)
   OnePoleHighPassFilter_0 = UE.One-Pole High Pass Filter:Audio(Cutoff Frequency=@n19.Out Value, In=UE.Reroute(Value=@n17.Out.R).Value)
   OnePoleLowPassFilter_1 = UE.One-Pole Low Pass Filter:Audio(In=@OnePoleHighPassFilter_0.Out, Cutoff Frequency=@n20.Value)
   Multiply_0 = UE.Multiply:Audio by Float(AdditionalOperands=^MasterVolume, PrimaryOperand=UE.Reroute(Value=@OnePoleLowPassFilter_1.Out).Value)
   AudioBusWriter1_0 = "UE.Audio Bus Writer (1)"(In 0=UE.Add:Audio(
     PrimaryOperand=UE.Multiply:Audio by Float(PrimaryOperand=@Multiply_3.Out, AdditionalOperands=@n21.Value).Out
     AdditionalOperands=UE.Multiply:Audio by Float(PrimaryOperand=@Multiply_0.Out, AdditionalOperands=@n21.Value).Out
    ).Out) {also {defaults {Audio Bus=/Game/Sounds/Modulation/EnvelopeFollowers/FortWeapon/AB_WeaponFire}}}
   InitVariable.Float()
   InitVariable.Trigger()
   InitVariable.Trigger()
   n22 = InitVariable.Trigger()
   TriggerAnyTriggerAny2_1 = "TriggerAny.Trigger Any (2)"(In 0=@n11.OnFinished, In 1=@n6.False)
   n23 = InitVariable.Trigger()
   TriggerAnyTriggerAny2_0 = "TriggerAny.Trigger Any (2)"(In 0=@n13.OnFinished, In 1=@n12.False)
   TriggerAccumulatorTriggerAccumulate2_0 = "TriggerAccumulator.Trigger Accumulate (2)"(
    In 0=VariableAccessor.Trigger(Variable=VariableMutator.Trigger(Variable=@n22.Variable, Value=@TriggerAnyTriggerAny2_1.Out).Variable).Value
    In 1=VariableAccessor.Trigger(Variable=VariableMutator.Trigger(Variable=@n23.Variable, Value=@TriggerAnyTriggerAny2_0.Out).Variable).Value
   ) {also {defaults {Auto Reset="<MetaSoundBuilderResult.SUCCEEDED: 0>"}}}
   n24 = &MSP_TriggerResolve(A=UE.Reroute(Value=@TriggerAccumulatorTriggerAccumulate2_0.Out).Value, B=@n3.Value)
   TriggerDelay_0 = UE.Trigger Delay(Delay Time=^RetriggerDelayTime, In=@TriggerAccumulatorTriggerAccumulate2_0.Out)
   TriggerControl_0 = UE.Trigger Control(Open=@n24.A, Close=@n24.B, Trigger In=@TriggerDelay_0.Out)
   ValueValue_0 = Value.Value:Bool(Set=@TriggerControl_0.Trigger Out)
   VariableMutator.Int32(Variable=InitVariable.Int32().Variable, Value=@n1.CurrentAmmo)
  ]
  links [
   {dst="@UE.OutputFormat.Stereo.Audio:1", src=Multiply_0.Out}
   {dst="@UE.OutputFormat.Stereo.Audio:0", src=Multiply_3.Out}
   {dst="@UE.Source.OneShot.OnFinished", src=TriggerControl_0.Trigger Out}
   {dst="@bIsAlmostFinished", src=ValueValue_0.Output Value}
  ]
 }]
properties {
 assetImportData=&self:AssetImportData
 numChannels=2
 outputFormat=Stereo
 qualitySetting=Fort_Weapons
 sampleRate=48000
 soundClassObject=&Fort_WeaponRanged_Fire_3P_PSM
 concurrencyOverrides {bLimitToOwner=true, maxCount=2, resolutionRule=StopOldest}
 concurrencySet [&Weapon_Gunfire_Global_Conc]
 modulationSettings {
  volumeRouting=Union
  volumeModulationDestination {bEnableModulation=true, modulators [/FortMetaSoundWeapon/Items/Modulation/PP_Weapon_Fire_3p_Vol]}
 }
}
