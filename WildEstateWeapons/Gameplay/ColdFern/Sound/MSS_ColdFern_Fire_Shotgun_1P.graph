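# One-shot 5.1 MetaSound source for the ColdFern shotgun's first-person fire sound.
# A shared 1P fire-control subgraph turns play, single-shot and firing inputs into gunshot triggers for a MainFire subgraph, which takes four waves, pitch/volume randomization and an aim-down-sights volume multiplier.
# Each channel is scaled by a mix volume; ammo, fire-rate and suppressor inputs are unused.
# Concurrency allows two instances, stopping the oldest, and a modulator controls volume.
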MetaSoundSource /WildEstateWeapons/Gameplay/ColdFern/Sound/MSS_ColdFern_Fire_Shotgun_1P
inputs [
 bIsFiring {type=Bool}
 FireSingleShot {type=Trigger}
 MainFire.ADSMainVolumeMultiply {type=Float}
 MainFire.Random.MaxPitch {type=Float}
 MainFire.Random.MaxVolume {type=Float}
 MainFire.Random.MinPitch {type=Float}
 MainFire.Random.MinVolume {type=Float}
 MainFire.Sounds {type=WaveAsset:Array}
 Mixing.MainFireVolume {type=Float}
 UE.Source.OnPlay {type=Trigger}
 Weapon.AmmoRemaining {type=Int32}
 Weapon.ClipSize {type=Int32}
 Weapon.FireRate {type=Float}
 Weapon.FireStartTime {type=Float}
 Weapon.IsAimingDownSights {type=Bool}
 Weapon.IsSuppressorEnabled {type=Bool}
]
interfaces [Fort.Weapon:1.0, UE.OutputFormat.5dot1:1.0, UE.Source.OneShot:1.0, UE.Source:1.1]
outputs [
 "UE.OutputFormat.5dot1.Audio:0" {type=Audio}
 "UE.OutputFormat.5dot1.Audio:1" {type=Audio}
 "UE.OutputFormat.5dot1.Audio:2" {type=Audio}
 "UE.OutputFormat.5dot1.Audio:3" {type=Audio}
 "UE.OutputFormat.5dot1.Audio:4" {type=Audio}
 "UE.OutputFormat.5dot1.Audio:5" {type=Audio}
 UE.Source.OneShot.OnFinished {type=Trigger}
]
pages [Default {
  input_defaults {
   FireSingleShot=null
   MainFire.ADSMainVolumeMultiply=1.3
   MainFire.Random.MaxPitch=1.05
   MainFire.Random.MaxVolume=1
   MainFire.Random.MinPitch=0.95
   MainFire.Random.MinVolume=0.9
   Mixing.MainFireVolume=1
   UE.Source.OnPlay=null
   Weapon.AmmoRemaining=30
   Weapon.ClipSize=30
   Weapon.FireRate=0.5
   Weapon.FireStartTime=0
   Weapon.IsAimingDownSights=null
   Weapon.IsSuppressorEnabled=null
   bIsFiring=null
   MainFire.Sounds [
    /WildEstateWeapons/Gameplay/ColdFern/Sound/wav/ColdFern_Wpn_SG_Fire_1P_01
    /WildEstateWeapons/Gameplay/ColdFern/Sound/wav/ColdFern_Wpn_SG_Fire_1P_02
    /WildEstateWeapons/Gameplay/ColdFern/Sound/wav/ColdFern_Wpn_SG_Fire_1P_03
    /WildEstateWeapons/Gameplay/ColdFern/Sound/wav/ColdFern_Wpn_SG_Fire_1P_04
   ]
  }
  nodes [
   MSP_Weapon_Fire_MainFire_0 = /FortMetaSoundWeapon/Items/Metasounds/Subgraphs/MSP_Weapon_Fire_MainFire(
    MainFire.ADS.VolumeMultiply=^MainFire.ADSMainVolumeMultiply
    MainFire.Random.Pitch.Max=^MainFire.Random.MaxPitch
    MainFire.Random.Volume.Max=^MainFire.Random.MaxVolume
    MainFire.Random.Pitch.Min=^MainFire.Random.MinPitch
    MainFire.Random.Volume.Min=^MainFire.Random.MinVolume
    MainFire.Sounds=^MainFire.Sounds
    Weapon.IsAimingDownSights=^Weapon.IsAimingDownSights
    OnGunshot=/FortMetaSoundWeapon/Items/Metasounds/Subgraphs/MSP_Weapon_Fire_Control_1p(FireSingleShot=^FireSingleShot, OnPlay=^UE.Source.OnPlay, bIsFiring=^bIsFiring).OutGunshot
   ) {also {defaults {bUseSingleVoice="<MetaSoundBuilderResult.SUCCEEDED: 0>"}}}
   Multiply_0 = UE.Multiply:Audio by Float(AdditionalOperands=^Mixing.MainFireVolume, PrimaryOperand=@MSP_Weapon_Fire_MainFire_0.MainFire.Rs)
   Multiply_1 = UE.Multiply:Audio by Float(AdditionalOperands=^Mixing.MainFireVolume, PrimaryOperand=@MSP_Weapon_Fire_MainFire_0.MainFire.C)
   Multiply_2 = UE.Multiply:Audio by Float(AdditionalOperands=^Mixing.MainFireVolume, PrimaryOperand=@MSP_Weapon_Fire_MainFire_0.MainFire.L)
   Multiply_3 = UE.Multiply:Audio by Float(AdditionalOperands=^Mixing.MainFireVolume, PrimaryOperand=@MSP_Weapon_Fire_MainFire_0.MainFire.R)
   Multiply_4 = UE.Multiply:Audio by Float(AdditionalOperands=^Mixing.MainFireVolume, PrimaryOperand=@MSP_Weapon_Fire_MainFire_0.MainFire.LFE)
   Multiply_5 = UE.Multiply:Audio by Float(AdditionalOperands=^Mixing.MainFireVolume, PrimaryOperand=@MSP_Weapon_Fire_MainFire_0.MainFire.Ls)
  ]
  links [
   {dst="@UE.Source.OneShot.OnFinished", src=MSP_Weapon_Fire_MainFire_0.OnFinished}
   {dst="@UE.OutputFormat.5dot1.Audio:5", src=Multiply_0.Out}
   {dst="@UE.OutputFormat.5dot1.Audio:2", src=Multiply_1.Out}
   {dst="@UE.OutputFormat.5dot1.Audio:0", src=Multiply_2.Out}
   {dst="@UE.OutputFormat.5dot1.Audio:1", src=Multiply_3.Out}
   {dst="@UE.OutputFormat.5dot1.Audio:3", src=Multiply_4.Out}
   {dst="@UE.OutputFormat.5dot1.Audio:4", src=Multiply_5.Out}
  ]
 }]
properties {
 assetImportData=&self:AssetImportData
 bOverrideConcurrency=true
 numChannels=6
 outputFormat=FiveDotOne
 qualitySetting=Fort_Weapons
 sampleRate=48000
 soundClassObject=&Fort_WeaponRanged_Fire_1P_PSM
 concurrencyOverrides {maxCount=2, resolutionRule=StopOldest}
 modulationSettings {
  volumeRouting=Union
  volumeModulationDestination {bEnableModulation=true, modulators [/FortMetaSoundWeapon/Items/Modulation/PP_Weapon_Fire_1p_Vol]}
 }
}
