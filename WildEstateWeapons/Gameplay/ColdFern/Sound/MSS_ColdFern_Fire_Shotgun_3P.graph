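# MetaSound source that plays the third-person ColdFern shotgun fire sound in stereo.
# Shared fire-control logic triggers random close stereo and distant mono layers, each scaled by a distance curve from a wave table bank over normalised attenuation distance.
# The summed layers take an affiliation-selected gain, using friendly or enemy modulators, then the weapon output volume.
# Finish fires once both layers end, and concurrency caps voices per owner, stopping the oldest.
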MetaSoundSource /WildEstateWeapons/Gameplay/ColdFern/Sound/MSS_ColdFern_Fire_Shotgun_3P
inputs [
 Character.Affiliation {type=Enum:FortAudioAffiliation}
 Character.Health {type=Float}
 Character.UniqueID {type=Int32}
 FireSingleShot {type=Trigger}
 MainFire.DistanceCurves {type=WaveTableBankAsset}
 MainFire.FalloffDistance {type=Float}
 MainFire.SoundsClose {type=WaveAsset:Array}
 MainFire.SoundsDistant {type=WaveAsset:Array}
 Mixing.CloseGain {type=Float}
 Mixing.DistantGain {type=Float}
 Mixing.Friendly.FriendlyFireWeaponVolume {type=Float}
 Mixing.WeaponOutputVolume {type=Float}
 StartFiring {type=Trigger}
 StopFiring {type=Trigger}
 UE.Attenuation.Distance {type=Float}
 UE.Source.OnPlay {type=Trigger}
]
interfaces [Fort.Character:1.0, UE.Attenuation:1.0, UE.OutputFormat.Stereo:1.0, UE.Source.OneShot:1.0, UE.Source:1.1]
outputs [
 "UE.OutputFormat.Stereo.Audio:0" {type=Audio}
 "UE.OutputFormat.Stereo.Audio:1" {type=Audio}
 UE.Source.OneShot.OnFinished {type=Trigger}
]
pages [Default {
  input_defaults {
   Character.Affiliation=0
   Character.Health=1
   Character.UniqueID=0
   FireSingleShot=null
   MainFire.DistanceCurves=&/ColdFern_DistanceCurves
   MainFire.FalloffDistance=27500
   Mixing.CloseGain=0.85
   Mixing.DistantGain=1
   Mixing.Friendly.FriendlyFireWeaponVolume=0.7
   Mixing.WeaponOutputVolume=1
   StartFiring=null
   StopFiring=null
   UE.Attenuation.Distance=25000
   UE.Source.OnPlay=null
   MainFire.SoundsClose [
    /WildEstateWeapons/Gameplay/ColdFern/Sound/wav/ColdFern_Wpn_SG_Fire_3P_Close_01
    /WildEstateWeapons/Gameplay/ColdFern/Sound/wav/ColdFern_Wpn_SG_Fire_3P_Close_02
    /WildEstateWeapons/Gameplay/ColdFern/Sound/wav/ColdFern_Wpn_SG_Fire_3P_Close_03
    /WildEstateWeapons/Gameplay/ColdFern/Sound/wav/ColdFern_Wpn_SG_Fire_3P_Close_04
   ]
   MainFire.SoundsDistant [
    /WildEstateWeapons/Gameplay/ColdFern/Sound/wav/ColdFern_Wpn_SG_Fire_3P_Dist_01
    /WildEstateWeapons/Gameplay/ColdFern/Sound/wav/ColdFern_Wpn_SG_Fire_3P_Dist_02
    /WildEstateWeapons/Gameplay/ColdFern/Sound/wav/ColdFern_Wpn_SG_Fire_3P_Dist_03
    /WildEstateWeapons/Gameplay/ColdFern/Sound/wav/ColdFern_Wpn_SG_Fire_3P_Dist_04
   ]
  }
  nodes [
   n1 = /FortMetaSoundWeapon/Items/Metasounds/Subgraphs/MSP_Weapon_Fire_Control_3p(FireSingleShot=^FireSingleShot, StartFiring=^StartFiring, StopFiring=^StopFiring)
   n2 = /FortMetaSoundWeapon/Items/Metasounds/Subgraphs/MSP_Weapon_Fire_3p_Mono(Sounds=^MainFire.SoundsDistant, OnGunshot=@n1.OutGunshot) {also {defaults {Random.Pitch.Max=1, Random.Pitch.Min=1, Random.Volume.Max=1.1, Random.Volume.Min=0.9}}}
   n3 = VariableAccessor.Float(Variable=VariableMutator.Float(
     Variable=InitVariable.Float().Variable
     Value=MapRange.MapRange:Float(In Range B=^MainFire.FalloffDistance, In=^UE.Attenuation.Distance) {also {defaults {Out Range A=0.001}}}.Out Value
    ).Variable)
   n4 = VariableAccessor.Float(Variable=@n3.Variable)
   n5 = UE.Multiply:Audio by Float(
    AdditionalOperands=^Mixing.DistantGain
    PrimaryOperand=UE.Multiply:Audio by Float(
     PrimaryOperand=@n2.Out.L
     AdditionalOperands=UE.WaveTableBankEvaluate(WaveTableBank=^MainFire.DistanceCurves, Input=@n4.Value) {also {defaults {Index=1}}}.Output
    ).Out
   )
   n6 = /FortMetaSoundWeapon/Items/Metasounds/Subgraphs/MSP_Weapon_Fire_3p_Stereo(Sounds=^MainFire.SoundsClose, OnGunshot=@n1.OutGunshot) {also {defaults {Random.Pitch.Max=1, Random.Pitch.Min=1, Random.Volume.Max=1.1, Random.Volume.Min=0.9}}}
   n7 = UE.WaveTableBankEvaluate(WaveTableBank=^MainFire.DistanceCurves, Input=@n3.Value)
   n8 = UE.Multiply:Audio by Float(AdditionalOperands=^Mixing.CloseGain, PrimaryOperand=UE.Multiply:Audio by Float(PrimaryOperand=@n6.Out.R, AdditionalOperands=@n7.Output).Out)
   n9 = UE.Multiply:Float(
    PrimaryOperand=^Mixing.Friendly.FriendlyFireWeaponVolume
    AdditionalOperands=Modulation.GetModulatorValue() {also {defaults {Modulator=/FortMetaSoundWeapon/Items/Modulation/CB_WeaponFire_3p_Friendly_Vol}}}.Out
   )
   CrossfadeTriggerRouteFloat4_0 = "Crossfade.Trigger Route (Float, 4):Float"(
    Crossfade Value=Convert.Float:Int32(Int32=Convert.Int32:Enum:FortAudioAffiliation("Enum:FortAudioAffiliation"=^Character.Affiliation).Int32).Float
    In 2=Modulation.GetModulatorValue() {also {defaults {Modulator=/FortMetaSoundWeapon/Items/Modulation/CB_WeaponFire_3p_Enemy_Vol}}}.Out
    In 1=@n9.Out
    In 3=@n9.Out
   ) {also {defaults {In 0=1}}}
   Multiply_5 = UE.Multiply:Audio by Float(
    AdditionalOperands=^Mixing.WeaponOutputVolume
    PrimaryOperand=UE.Multiply:Audio by Float(PrimaryOperand=UE.Add:Audio(AdditionalOperands=@n5.Out, PrimaryOperand=@n8.Out).Out, AdditionalOperands=@CrossfadeTriggerRouteFloat4_0.Out).Out
   )
   n10 = UE.Multiply:Audio by Float(AdditionalOperands=^Mixing.CloseGain, PrimaryOperand=UE.Multiply:Audio by Float(PrimaryOperand=@n6.Out.L, AdditionalOperands=@n7.Output).Out)
   Multiply_7 = UE.Multiply:Audio by Float(
    AdditionalOperands=^Mixing.WeaponOutputVolume
    PrimaryOperand=UE.Multiply:Audio by Float(PrimaryOperand=UE.Add:Audio(PrimaryOperand=@n10.Out, AdditionalOperands=@n5.Out).Out, AdditionalOperands=@CrossfadeTriggerRouteFloat4_0.Out).Out
   )
   n11 = InitVariable.Trigger()
   n12 = InitVariable.Trigger()
   TriggerAccumulatorTriggerAccumulate2_0 = "TriggerAccumulator.Trigger Accumulate (2)"(
    In 1=VariableAccessor.Trigger(Variable=VariableMutator.Trigger(Variable=@n11.Variable, Value=@n2.OnFinished).Variable).Value
    In 0=VariableAccessor.Trigger(Variable=VariableMutator.Trigger(Variable=@n12.Variable, Value=@n6.OnFinished).Variable).Value
   ) {also {defaults {Auto Reset="<MetaSoundBuilderResult.SUCCEEDED: 0>"}}}
  ]
  links [
   {dst="@UE.OutputFormat.Stereo.Audio:1", src=Multiply_5.Out}
   {dst="@UE.OutputFormat.Stereo.Audio:0", src=Multiply_7.Out}
   {dst="@UE.Source.OneShot.OnFinished", src=TriggerAccumulatorTriggerAccumulate2_0.Out}
  ]
 }]
properties {
 assetImportData=&self:AssetImportData
 attenuationSettings=&/ATT_ColdFern_3P
 bOverrideConcurrency=true
 maxDistance=27500
 numChannels=2
 outputFormat=Stereo
 qualitySetting=Fort_Weapons
 sampleRate=48000
 soundClassObject=&Fort_WeaponRanged_Fire_3P_PSM
 concurrencyOverrides {bLimitToOwner=true, maxCount=3, resolutionRule=StopOldest}
 modulationSettings {
  volumeRouting=Union
  volumeModulationDestination {bEnableModulation=true, modulators [/FortMetaSoundWeapon/Items/Modulation/PP_Weapon_Fire_3p_Vol]}
 }
}
