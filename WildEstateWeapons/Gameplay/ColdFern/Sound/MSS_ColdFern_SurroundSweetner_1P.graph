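# First-person 5.1 MetaSound sweetener for ColdFern weapon fire, mixing surround first-shot blast, shell and "silly" layers.
# Play starts ADSR-shaped shell and randomly offset silly loops, plus a blast if DoSpecialShot passes; OnStop releases loops and triggers random tails.
# Shells sit front-right and side-right, silly layers in front left/right, with the enveloped silly loop also feeding an audio bus.
# The Mobile page is silent, finishing immediately.
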MetaSoundSource /WildEstateWeapons/Gameplay/ColdFern/Sound/MSS_ColdFern_SurroundSweetner_1P
inputs [
 Audio Bus {type=AudioBusAsset}
 Character.Affiliation {type=Enum:FortAudioAffiliation}
 Character.Health {type=Float}
 Character.UniqueID {type=Int32}
 DoSpecialShot {type=Bool}
 Multiplier_Quad {type=Float}
 Multiplier_Shells {type=Float}
 Multiplier_Silly_Loop {type=Float}
 Multiplier_Silly_Loop_End {type=Float}
 OnStop {type=Trigger}
 Shells Loop {type=WaveAsset:Array}
 Shells Loop End {type=WaveAsset:Array}
 Silly Loop {type=WaveAsset:Array}
 Silly Loop End {type=WaveAsset:Array}
 SurroundAssets {type=WaveAsset:Array}
 UE.Source.OnPlay {type=Trigger}
]
interfaces [Fort.Character:1.0, UE.OutputFormat.5dot1:1.0, UE.Source.OneShot:1.0, UE.Source:1.1]
outputs [
 "UE.OutputFormat.5dot1.Audio:0" {type=Audio}
 "UE.OutputFormat.5dot1.Audio:1" {type=Audio}
 "UE.OutputFormat.5dot1.Audio:2" {type=Audio}
 "UE.OutputFormat.5dot1.Audio:3" {type=Audio}
 "UE.OutputFormat.5dot1.Audio:4" {type=Audio}
 "UE.OutputFormat.5dot1.Audio:5" {type=Audio}
 UE.Source.OneShot.OnFinished {type=Trigger}
]
pages [
 Default {
  is_build_graph=true
  input_defaults {
   Audio Bus=/WildEstateWeapons/Gameplay/ColdFern/Sound/Modulation/AB_ColdFern
   Character.Affiliation=0
   Character.Health=0
   Character.UniqueID=0
   DoSpecialShot="<MetaSoundBuilderResult.SUCCEEDED: 0>"
   Multiplier_Quad=0.508
   Multiplier_Shells=0.25
   Multiplier_Silly_Loop=0.8
   Multiplier_Silly_Loop_End=0.3
   OnStop=null
   UE.Source.OnPlay=null
   Shells Loop [/WildEstateWeapons/Gameplay/ColdFern/Sound/wav/ColdFern_Wpn_Shell_Loop_01]
   Shells Loop End [
    /WildEstateWeapons/Gameplay/ColdFern/Sound/wav/ColdFern_Wpn_Shell_Loop_End_01
    /WildEstateWeapons/Gameplay/ColdFern/Sound/wav/ColdFern_Wpn_Shell_Loop_End_02
    /WildEstateWeapons/Gameplay/ColdFern/Sound/wav/ColdFern_Wpn_Shell_Loop_End_03
    /WildEstateWeapons/Gameplay/ColdFern/Sound/wav/ColdFern_Wpn_Shell_Loop_End_04
    /WildEstateWeapons/Gameplay/ColdFern/Sound/wav/ColdFern_Wpn_Shell_Loop_End_05
   ]
   Silly Loop [/WildEstateWeapons/Gameplay/ColdFern/Sound/wav/ColdFern_Wpn_Silly_Loop_01]
   Silly Loop End [
    /WildEstateWeapons/Gameplay/ColdFern/Sound/wav/ColdFern_Wpn_Silly_Loop_End_01
    /WildEstateWeapons/Gameplay/ColdFern/Sound/wav/ColdFern_Wpn_Silly_Loop_End_02
    /WildEstateWeapons/Gameplay/ColdFern/Sound/wav/ColdFern_Wpn_Silly_Loop_End_03
    /WildEstateWeapons/Gameplay/ColdFern/Sound/wav/ColdFern_Wpn_Silly_Loop_End_04
    /WildEstateWeapons/Gameplay/ColdFern/Sound/wav/ColdFern_Wpn_Silly_Loop_End_05
    /WildEstateWeapons/Gameplay/ColdFern/Sound/wav/ColdFern_Wpn_Silly_Loop_End_06
    /WildEstateWeapons/Gameplay/ColdFern/Sound/wav/ColdFern_Wpn_Silly_Loop_End_07
    /WildEstateWeapons/Gameplay/ColdFern/Sound/wav/ColdFern_Wpn_Silly_Loop_End_08
   ]
   SurroundAssets [
    /WildEstateWeapons/Gameplay/ColdFern/Sound/wav/ColdFern_Wpn_FirstShotBlast_01
    /WildEstateWeapons/Gameplay/ColdFern/Sound/wav/ColdFern_Wpn_FirstShotBlast_02
   ]
  }
  nodes [
   n1 = UE.Reroute(Value=^UE.Source.OnPlay)
   ArrayRandomGet_3 = Array.Random Get:WaveAsset:Array(In Array=^Silly Loop, Next=@n1.Value) {also {defaults {No Repeats=9, Weights []}}}
   ADSREnvelopeADSREnvelope_0 = ADSR Envelope.ADSR Envelope:Audio(Trigger Release=^OnStop, Trigger Attack=@ArrayRandomGet_3.On Next) {also {defaults {Attack Time=0.3, Decay Time=2, Release Time=0.2, Sustain Level=0.65}}}
   n2 = UE.Reroute(Value=UE.Reroute(Value=@ADSREnvelopeADSREnvelope_0.Out Envelope).Value)
   WavePlayer_1 = UE.Wave Player:Stereo(
    Play=@ADSREnvelopeADSREnvelope_0.On Attack Triggered
    Stop=@ADSREnvelopeADSREnvelope_0.On Done
    Wave Asset=@ArrayRandomGet_3.Value
    Start Time=UE.ConversionFloatToTime(In=/FortMetasound/Patches/RandomFloatValueByTrigger(TriggerIn=@ADSREnvelopeADSREnvelope_0.On Attack Triggered) {also {defaults {FloatMax=5}}}.Output).Out
   ) {also {defaults {Loop="<MetaSoundBuilderResult.SUCCEEDED: 0>"}, version [1,1]}}
   n3 = UE.Multiply:Audio(AdditionalOperands=@n2.Value, PrimaryOperand=@WavePlayer_1.Out Left)
   n4 = UE.Multiply:Audio(PrimaryOperand=@n2.Value, AdditionalOperands=@WavePlayer_1.Out Right)
   AudioBusWriter2_0 = "UE.Audio Bus Writer (2)"(Audio Bus=^Audio Bus, In 0=@n3.Out, In 1=@n4.Out)
   n5 = TriggerCompare.Clamp:Bool(A=^DoSpecialShot, Compare=@n1.Value) {also {defaults {B="<MetaSoundBuilderResult.SUCCEEDED: 0>"}}}
   ArrayRandomGet_2 = Array.Random Get:WaveAsset:Array(In Array=^SurroundAssets, Next=@n5.True)
   WavePlayer_0 = UE.Wave Player:5dot1(Play=@ArrayRandomGet_2.On Next, Wave Asset=@ArrayRandomGet_2.Value) {also {version [1,1]}}
   ArrayRandomGet_4 = Array.Random Get:WaveAsset:Array(Next=^OnStop, In Array=^Shells Loop End) {also {defaults {Weights []}}}
   WavePlayer_3 = UE.Wave Player:Stereo(Play=@ArrayRandomGet_4.On Next, Wave Asset=@ArrayRandomGet_4.Value) {also {version [1,1]}}
   ArrayRandomGet_0 = Array.Random Get:WaveAsset:Array(In Array=^Shells Loop, Next=@n1.Value) {also {defaults {Weights []}}}
   ADSREnvelopeADSREnvelope_1 = ADSR Envelope.ADSR Envelope:Audio(Trigger Release=^OnStop, Trigger Attack=@ArrayRandomGet_0.On Next) {also {defaults {Attack Time=0.3, Release Time=0.2, Sustain Level=1}}}
   n6 = UE.Reroute(Value=UE.Reroute(Value=@ADSREnvelopeADSREnvelope_1.Out Envelope).Value)
   WavePlayer_4 = UE.Wave Player:Stereo(Play=@ADSREnvelopeADSREnvelope_1.On Attack Triggered, Stop=@ADSREnvelopeADSREnvelope_1.On Done, Wave Asset=@ArrayRandomGet_0.Value) {also {defaults {Loop="<MetaSoundBuilderResult.SUCCEEDED: 0>"}, version [1,1]}}
   n7 = UE.Reroute(Value=@WavePlayer_4.Out Right)
   AudioMixerAudioMixerMono3_0 = "AudioMixer.Audio Mixer (Mono, 3)"(
    In 0=UE.Multiply:Audio by Float(AdditionalOperands=^Multiplier_Quad, PrimaryOperand=@WavePlayer_0.Out Side Right).Out
    In 2=UE.Multiply:Audio by Float(AdditionalOperands=^Multiplier_Shells, PrimaryOperand=@WavePlayer_3.Out Right).Out
    In 1=UE.Reroute(Value=UE.Multiply:Audio(AdditionalOperands=@n6.Value, PrimaryOperand=@n7.Value).Out).Value
   )
   ArrayRandomGet_1 = Array.Random Get:WaveAsset:Array(Next=^OnStop, In Array=^Silly Loop End) {also {defaults {No Repeats=-1, Weights []}}}
   WavePlayer_2 = UE.Wave Player:Stereo(Play=@ArrayRandomGet_1.On Next, Wave Asset=@ArrayRandomGet_1.Value) {also {version [1,1]}}
   AudioMixerAudioMixerMono3_1 = "AudioMixer.Audio Mixer (Mono, 3)"(
    Gain 1=^Multiplier_Silly_Loop
    Gain 2=^Multiplier_Silly_Loop_End
    In 1=@n3.Out
    In 0=UE.Multiply:Audio by Float(AdditionalOperands=^Multiplier_Quad, PrimaryOperand=@WavePlayer_0.Out Front Left).Out
    In 2=@WavePlayer_2.Out Left
   )
   AudioMixerAudioMixerMono5_0 = "AudioMixer.Audio Mixer (Mono, 5)"(
    Gain 3=^Multiplier_Silly_Loop
    Gain 4=^Multiplier_Silly_Loop_End
    In 0=UE.Multiply:Audio by Float(AdditionalOperands=^Multiplier_Quad, PrimaryOperand=@WavePlayer_0.Out Front Right).Out
    In 2=UE.Multiply:Audio by Float(AdditionalOperands=^Multiplier_Shells, PrimaryOperand=@WavePlayer_3.Out Left).Out
    In 1=UE.Reroute(Value=UE.Multiply:Audio(AdditionalOperands=@n6.Value, PrimaryOperand=@WavePlayer_4.Out Left).Out).Value
    In 4=UE.Reroute(Value=@WavePlayer_2.Out Right).Value
    In 3=UE.Reroute(Value=@n4.Out).Value
   )
   Multiply_5 = UE.Multiply:Audio by Float(AdditionalOperands=^Multiplier_Quad, PrimaryOperand=@WavePlayer_0.Out Front Center)
   Multiply_7 = UE.Multiply:Audio by Float(AdditionalOperands=^Multiplier_Quad, PrimaryOperand=@WavePlayer_0.Out Side Left)
   TriggerAnyTriggerAny2_0 = "TriggerAny.Trigger Any (2)"(In 0=UE.Reroute(Value=UE.Reroute(Value=@n5.False).Value).Value, In 1=@WavePlayer_0.On Finished)
   TriggerAccumulatorTriggerAccumulate5_0 = "TriggerAccumulator.Trigger Accumulate (5)"(
    In 2=UE.Reroute(Value=@WavePlayer_3.On Finished).Value
    In 1=UE.Reroute(Value=@WavePlayer_4.On Finished).Value
    In 0=@TriggerAnyTriggerAny2_0.Out
    In 3=@WavePlayer_1.On Finished
    In 4=@WavePlayer_2.On Finished
   )
  ]
  links [
   {dst="@UE.OutputFormat.5dot1.Audio:5", src=AudioMixerAudioMixerMono3_0.Out}
   {dst="@UE.OutputFormat.5dot1.Audio:0", src=AudioMixerAudioMixerMono3_1.Out}
   {dst="@UE.OutputFormat.5dot1.Audio:1", src=AudioMixerAudioMixerMono5_0.Out}
   {dst="@UE.OutputFormat.5dot1.Audio:2", src=Multiply_5.Out}
   {dst="@UE.OutputFormat.5dot1.Audio:4", src=Multiply_7.Out}
   {dst="@UE.Source.OneShot.OnFinished", src=TriggerAccumulatorTriggerAccumulate5_0.Out}
   {dst="@UE.OutputFormat.5dot1.Audio:3", src=WavePlayer_0.Out Low Frequency}
  ]
 }
 Mobile {links [{dst="@UE.Source.OneShot.OnFinished", src="@UE.Source.OnPlay"}], nodes []}
]
properties {
 assetImportData=&self:AssetImportData
 bOverrideConcurrency=true
 numChannels=6
 outputFormat=FiveDotOne
 sampleRate=48000
 soundClassObject=&Fort_WeaponRanged_Fire_1P_PSM
 concurrencyOverrides {maxCount=2}
}
